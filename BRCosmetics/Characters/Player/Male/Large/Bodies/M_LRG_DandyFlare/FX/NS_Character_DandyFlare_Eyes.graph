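# Niagara system that adds glowing red eye sprites to the DandyFlare character.
# One emitter bursts two spawn groups of two persistent sprites, using the spawn group as visibility tag across two sprite renderers and to select size and offset.
# Particles follow two sockets through a socket reader, chosen by partition index.
# Color comes from a time-of-day function set to the same red throughout, lerped to transparent when the head is hidden.
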NiagaraSystem /BRCosmetics/Characters/Player/Male/Large/Bodies/M_LRG_DandyFlare/FX/NS_Character_DandyFlare_Eyes
dependencies {
 usedDataInterfaces [&NiagaraDataInterfaceArrayInt32, &NiagaraDataInterfaceSocketReader]
 usedDynamicInputs []
 usedModules [
  &self:SystemSpawnScript.SystemScriptSource.SystemScriptGraph.NiagaraNodeAssignment_0.SetVariables_#1
  &SystemState
 ]
}
emitters {NE_EyesGlow {
  emitterUpdateScript [
   &EmitterState(
    Inactive Response=::NewEnumerator1
    Life Cycle Mode=::NewEnumerator1
    Loop Behavior=::NewEnumerator1
    Loop Duration Mode=::NewEnumerator1
    UseLoopDelay=false
   )
   &SpawnBurst_Instantaneous(Spawn Count=2, Use Spawn Probability=false, UseLoopCountLimit=true)
   &SpawnBurst_Instantaneous(Spawn Count=2, Spawn Group=1, Use Spawn Probability=false, UseLoopCountLimit=true) {module=SpawnBurst_Instantaneous001}
  ]
  particleSpawnScript [
   &PartitionParticles()
   /SetVariables(Particles.VisibilityTag=&SpawnGroup()) {isSetParameters=true}
   &InitializeParticle(
    Color Mode=::NewEnumerator0
    Sprite Size Mode=::NewEnumerator3
    Uniform Sprite Size=&MergeHelper_FloatSwitch(0=10, 1=10, 2=?, 3=?, Index=$Particles.VisibilityTag, Num=::NewEnumerator0)
    UsePositionOffset=false
    Write Material Random=true
   )
   &CameraOffset()
  ]
  particleUpdateScript [
   &ParticleState(Kill Particles When Lifetime Has Elapsed=false)
   &SocketLocation(
    SamplePosOffset=true
    Sampled Position Offset=&MergeHelper_VectorSwitch(0=<0 0 0>, 1=<1 -4 0>, 2=?, 3=?, Index=$Particles.VisibilityTag, Num=::NewEnumerator0)
    Socket Index=&SelectIntFromArray(Direct Array Index=$Particles.PartitionParticles.Partition, Int Selection Array=DataInterface() {propertyValues {IntData [0,1]}})
    Socket Reader=$System.Socket Reader
    Socket Sampling Mode=::NewEnumerator2
    VelocitySpeedLimit=false
   )
   &ScaleColor(
    Color Scale:LinearColor=&LerpLinearColors(
     StartColor=&FNTimeOfDayColor(Night Color=<rgba 2 0 0 1>, Morning Color=<rgba 2 0 0 1>, Day Color=<rgba 2 0 0 1>, Evening Color=<rgba 2 0 0 1>)
     EndColor=<rgba 0 0 0 0>
     LerpFactor=&MakeCustomFloatFromBool(Bool=$User.bisHeadHidden, True Float=1, False Float=0)
    )
    Scale Mode=3
   )
  ]
  rendererClasses [&NiagaraSpriteRendererProperties]
  renderers [
   {rendererClass=&NiagaraSpriteRendererProperties}
   {rendererClass=&NiagaraSpriteRendererProperties, rendererIndex=1}
  ]
 }}
summary {
 systemName=NS_Character_DandyFlare_Eyes
 emitters [{emitterName=NE_EyesGlow, rendererClasses [&NiagaraSpriteRendererProperties]}]
 userVariables [
  User.bisFootHidden {type=NiagaraBool}
  User.bIsFrontEnd {type=NiagaraBool}
  User.bIsFrontEndPreview {type=NiagaraBool}
  User.bisHandHidden {type=NiagaraBool}
  User.bisHeadHidden {type=NiagaraBool}
  User.IsSkydiving {type=NiagaraBool}
  User.VariantID {type=NiagaraFloat, defaultValue=0, description=""}
 ]
}
