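# Material function for the DandyFlare character's emote disc, outputting an emission gain and area mask to drive animated lights.
# A UV-distance ring mask and an MF_RadialMask pattern, blendable toward its squared inverse, confine the glow to parts of the disc.
# Time drives the pattern's rotation through a looping step range, with optional flooring for stepped motion.
# Emote | Disc parameters expose shape, timing and light gain.
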MaterialFunction /BRCosmetics/Characters/Player/Male/Large/Bodies/M_LRG_DandyFlare/Materials/MaterialFunctions/MF_DandyFlare_Disc_Emissive
graph {nodes [
  n1 = NamedRerouteUsage()
  n2 = Multiply(A=NamedRerouteUsage(), B=NamedRerouteUsage())
  n3 = NamedRerouteUsage()
  FunctionOutput(None=LinearInterpolate(B=Multiply(A=@n1, B=@n2), Alpha=@n3) {constA=1}) {outputName="Emission Gain (V3)"}
  FunctionOutput(None=@n3) {outputName="Area Mask (S)", sortPriority=10}
  n4 = MaterialFunctionCall!(
   UV=NamedRerouteUsage()
   Pivot=Constant()
   Num Directions=NamedRerouteUsage()
   Spread=NamedRerouteUsage()
   Gradient=NamedRerouteUsage()
   Orientation=Constant() {r=-0.25}
   Rotation=NamedRerouteUsage()
  ) {materialFunction=&/MF_RadialMask}
  n5 = OneMinus(None=@n4."Mask (S)")
  NamedRerouteDeclaration(None=LinearInterpolate(A=@n4."Mask (S)", B=Multiply(A=@n5, B=@n5), Alpha=NamedRerouteUsage())) {name="Radial Mask (S)", nodeColor=<rgba 1 0 0.8672 1>}
  n6 = VectorParameter() {
   group=Emote | Disc
   parameterName=Disc Mask
   sortPriority=20
   channelNames {a=Radius Max, b=Radius Min, g=Disc Center V, r=Disc Center U}
   defaultValue=<rgba 0.735 0.728 0.171 0.254>
  }
  NamedRerouteDeclaration(None=AppendVector(A=@n6.Disc Center U, B=@n6.Disc Center U)) {name="Disc UV Center (V2)", nodeColor=<rgba 0 0.2969 1 1>}
  n7 = ScalarParameter() {defaultValue=2, group=Emote | Disc, parameterName=Mask Num Directions, sliderMax=16, sliderMin=1, sortPriority=30}
  NamedRerouteDeclaration(None=@n7) {name="Mask Num Directions (S)", nodeColor=<rgba 0.0781 0 1 1>}
  NamedRerouteDeclaration(None=ScalarParameter() {group=Emote | Disc, parameterName=Mask Invert, sliderMax=1, sortPriority=60}) {name="Mask Invert (S)", nodeColor=<rgba 0.3828 1 0 1>}
  n8 = VectorParameter() {
   group=Emote | Disc
   parameterName=Light Emission Gain
   sortPriority=10
   channelNames {a=Intensity, b=Tint B, g=Tint G, r=Tint R}
   defaultValue=<rgba 1 1 1 1>
  }
  NamedRerouteDeclaration(None=Multiply(A=@n8.RGB, B=@n8.RGB)) {name="Lights Emission Gain (V3)", nodeColor=<rgba 1 0 0.6562 1>}
  n9 = TextureCoordinate()
  n10 = NamedRerouteUsage()
  n11 = Length(None=NamedRerouteDeclaration(None=Subtract(A=@n9, B=@n10) {constB=0.5}) {name="Disc UV (Centered) (V2)", nodeColor=<rgba 0 1 0.5781 1>}) {is=class MaterialExpressionLength}
  n12 = Floor(None=Divide(A=@n11, B=NamedRerouteUsage()))
  n13 = NamedRerouteUsage()
  n14 = Saturate(None=Floor(None=Divide(A=@n11, B=@n13)))
  NamedRerouteDeclaration(None=Multiply(A=Saturate(None=@n12), B=OneMinus(None=@n14))) {name="Ring Mask (S)", nodeColor=<rgba 0.8984 0 1 1>}
  NamedRerouteDeclaration(None=ScalarParameter() {defaultValue=0.1, group=Emote | Disc, parameterName="Loop Duration (seconds)", sortPriority=70}) {name="Loop Duration (S)", nodeColor=<rgba 0 1 0.9297 1>}
  NamedRerouteDeclaration(None=ScalarParameter() {defaultValue=16, group=Emote | Disc, parameterName=Loop Num Steps, sortPriority=80}) {name="Loop Num Steps (S)", nodeColor=<rgba 1 0.6094 0 1>}
  NamedRerouteDeclaration(None=ScalarParameter() {defaultValue=2, group=Emote | Disc, parameterName=Loop Start Step, sortPriority=90}) {name="Loop Start Step (S)", nodeColor=<rgba 0.2891 0 1 1>}
  NamedRerouteDeclaration(None=ScalarParameter() {defaultValue=16, group=Emote | Disc, parameterName=Loop End Step, sortPriority=100}) {name="Loop End Step (S)", nodeColor=<rgba 0.0312 1 0 1>}
  NamedRerouteDeclaration(None=ScalarParameter() {defaultValue=1, group=Emote | Disc, parameterName=Floor Steps, sliderMax=1, sortPriority=120}) {name="Floor Steps (S)", nodeColor=<rgba 1 0 0.3516 1>}
  NamedRerouteDeclaration(None=ScalarParameter() {group=Emote | Disc, parameterName=Steps Offset, sortPriority=110}) {name="Steps Offset (S)", nodeColor=<rgba 0 0.6719 1 1>}
  NamedRerouteDeclaration(None=@n6.Radius Min) {name="Disc Radius Min (S)", nodeColor=<rgba 1 0.2578 0 1>}
  NamedRerouteDeclaration(None=@n6.Radius Max) {name="Disc Radius Max (S)", nodeColor=<rgba 0 0.0625 1 1>}
  NamedRerouteDeclaration(None=ScalarParameter() {defaultValue=0.0625, group=Emote | Disc, parameterName=Mask Spread, sliderMax=1, sortPriority=40}) {name="Mask Spread (S)", nodeColor=<rgba 0.3828 1 0 1>}
  n15 = Divide(A=Time(), B=NamedRerouteUsage())
  n16 = NamedRerouteUsage()
  n17 = Add(A=Fmod(A=Add(A=@n15, B=@n16), B=NamedRerouteUsage()), B=NamedRerouteUsage())
  n18 = Floor(None=@n17)
  n19 = NamedRerouteUsage()
  NamedRerouteDeclaration(None=Divide(A=LinearInterpolate(A=@n17, B=@n18, Alpha=@n19), B=NamedRerouteUsage()) {constB=16}) {name="Mask Rotation (S)", nodeColor=<rgba 1 0 0.7031 1>}
  NamedRerouteDeclaration(None=ScalarParameter() {defaultValue=0.05, group=Emote | Disc, parameterName=Mask Gradient, sliderMax=1, sortPriority=50}) {name="Mask Gradient (S)", nodeColor=<rgba 0 1 0.9766 1>}
 ]}
