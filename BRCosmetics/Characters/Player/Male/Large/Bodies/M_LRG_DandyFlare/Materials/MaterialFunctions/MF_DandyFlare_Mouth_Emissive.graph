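# Material function for the DandyFlare character's emote mouth, outputting an animated emissive gain and an area mask.
# A UV rectangle parameter is remapped to mirrored, centered coordinates that build the hard mask and a soft smoothstep gradient along U.
# An Anim Curve input, blendable with a debug parameter, feeds the animation, with controls for edge softness, off-state gain and inversion.
# Gain blends from 1 to the animated value through a mask.
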MaterialFunction /BRCosmetics/Characters/Player/Male/Large/Bodies/M_LRG_DandyFlare/Materials/MaterialFunctions/MF_DandyFlare_Mouth_Emissive
graph {expressions {
  MaterialExpressionAbs_1 {class=MaterialExpressionAbs, inputs [None {from=MaterialExpressionConstantBiasScale_0, output=""}], properties {}}
  MaterialExpressionAppendVector_4 {
   class=MaterialExpressionAppendVector
   inputs [
    A {from=MaterialExpressionVectorParameter_3, output=Min U}
    B {from=MaterialExpressionVectorParameter_3, output=Min U}
   ]
  }
  MaterialExpressionAppendVector_5 {
   class=MaterialExpressionAppendVector
   inputs [
    A {from=MaterialExpressionVectorParameter_3, output=Max U}
    B {from=MaterialExpressionVectorParameter_3, output=Max U}
   ]
  }
  MaterialExpressionComponentMask_3 {class=MaterialExpressionComponentMask, inputs [None {from=MaterialExpressionSwitch_1}], properties {r=true}}
  MaterialExpressionConstantBiasScale_0 {class=MaterialExpressionConstantBiasScale, inputs [None {from=MaterialExpressionDivide_3}], properties {bias=-0.5, desc="[-1, 1]", scale=2}}
  MaterialExpressionConvert_6 {
   class=MaterialExpressionConvert
   inputs [None {from=MaterialExpressionStep_0}]
   properties {
    nodeName=Break Float2
    convertInputs [{
      type=Vector2
      expressionInput {
       expression=&self:MaterialExpressionStep_0
       inputName=None
       mask=0
       maskA=0
       maskB=0
       maskG=0
       maskR=0
       outputIndex=0
      }
     }]
    convertOutputs [{type=Scalar}, {type=Scalar}]
   }
  }
  MaterialExpressionDivide_3 {
   class=MaterialExpressionDivide
   inputs [
    A {from=MaterialExpressionSubtract_1}
    B {from=MaterialExpressionSubtract_3}
   ]
   properties {desc="Mouth UV Centered [0, 1]"}
  }
  MaterialExpressionFunctionInput_0 {
   class=MaterialExpressionFunctionInput
   inputs [Preview {from="", output=""}]
   properties {bUsePreviewValueAsDefault=true, inputName=Anim Curve, inputType=FunctionInput_Scalar}
  }
  MaterialExpressionFunctionOutput_0 {
   class=MaterialExpressionFunctionOutput
   inputs [None {from=MaterialExpressionLinearInterpolate_8, output=""}]
   properties {outputName="Emissive Gain (S)"}
  }
  MaterialExpressionFunctionOutput_1 {
   class=MaterialExpressionFunctionOutput
   inputs [None {from=MaterialExpressionNamedRerouteUsage_2, output=""}]
   properties {outputName="Area Mask (S)", sortPriority=10}
  }
  MaterialExpressionLinearInterpolate_0 {
   class=MaterialExpressionLinearInterpolate
   inputs [
    A {from=MaterialExpressionFunctionInput_0}
    B {from=MaterialExpressionScalarParameter_34}
    Alpha {from=MaterialExpressionSaturate_0}
   ]
  }
  MaterialExpressionLinearInterpolate_8 {
   class=MaterialExpressionLinearInterpolate
   inputs [
    A {from=""}
    B {from=MaterialExpressionNamedRerouteUsage_3}
    Alpha {from=MaterialExpressionNamedRerouteUsage_2}
   ]
   properties {constA=1}
  }
  MaterialExpressionMultiply_11 {
   class=MaterialExpressionMultiply
   inputs [
    A {from=MaterialExpressionSmoothStep_4}
    B {from=MaterialExpressionOneMinus_6}
   ]
  }
  MaterialExpressionMultiply_7 {
   class=MaterialExpressionMultiply
   inputs [
    A {from=MaterialExpressionOneMinus_3}
    B {from=MaterialExpressionOneMinus_4}
   ]
  }
  MaterialExpressionNamedRerouteDeclaration_0 {
   class=MaterialExpressionNamedRerouteDeclaration
   inputs [None {from=MaterialExpressionAbs_1}]
   properties {name="Mouth Abs Centered UV (V2)", nodeColor=<rgba 0 1 0.6719 1>}
  }
  MaterialExpressionNamedRerouteDeclaration_1 {
   class=MaterialExpressionNamedRerouteDeclaration
   inputs [None {from=MaterialExpressionMultiply_7}]
   properties {name="Mouth Area Mask (S)", nodeColor=<rgba 1 0 0.7031 1>}
  }
  MaterialExpressionNamedRerouteDeclaration_2 {
   class=MaterialExpressionNamedRerouteDeclaration
   inputs [None {from=MaterialExpressionSwitch_2}]
   properties {name="Animated Emission Gain (S)", nodeColor=<rgba 0 1 0.9766 1>}
  }
  MaterialExpressionNamedRerouteDeclaration_23 {
   class=MaterialExpressionNamedRerouteDeclaration
   inputs [None {from=MaterialExpressionSaturate_7}]
   properties {name="Mouth Inverse Emission (S)", nodeColor=<rgba 1 0.6328 0 1>}
  }
  MaterialExpressionNamedRerouteDeclaration_34 {
   class=MaterialExpressionNamedRerouteDeclaration
   inputs [None {from=MaterialExpressionAppendVector_4}]
   properties {name="Mouth UV Min (V2)", nodeColor=<rgba 1 0.6797 0 1>}
  }
  MaterialExpressionNamedRerouteDeclaration_35 {
   class=MaterialExpressionNamedRerouteDeclaration
   inputs [None {from=MaterialExpressionAppendVector_5}]
   properties {name="Mouth UV Max (V2)", nodeColor=<rgba 0.3594 0 1 1>}
  }
  MaterialExpressionNamedRerouteDeclaration_42 {
   class=MaterialExpressionNamedRerouteDeclaration
   inputs [None {from=MaterialExpressionScalarParameter_35}]
   properties {name="Mouth Edge Softness (S)", nodeColor=<rgba 0 1 0.2266 1>}
  }
  MaterialExpressionNamedRerouteDeclaration_43 {
   class=MaterialExpressionNamedRerouteDeclaration
   inputs [None {from=MaterialExpressionScalarParameter_36}]
   properties {name="Mouth Min Emission (S)", nodeColor=<rgba 1 0 0.0938 1>}
  }
  MaterialExpressionNamedRerouteDeclaration_44 {
   class=MaterialExpressionNamedRerouteDeclaration
   inputs [None {from=MaterialExpressionSaturate_6}]
   properties {name="Anim Curve (S)", nodeColor=<rgba 0 1 0.6953 1>}
  }
  MaterialExpressionNamedRerouteUsage_0 {class=MaterialExpressionNamedRerouteUsage}
  MaterialExpressionNamedRerouteUsage_1 {class=MaterialExpressionNamedRerouteUsage}
  MaterialExpressionNamedRerouteUsage_2 {class=MaterialExpressionNamedRerouteUsage}
  MaterialExpressionNamedRerouteUsage_27 {class=MaterialExpressionNamedRerouteUsage}
  MaterialExpressionNamedRerouteUsage_29 {class=MaterialExpressionNamedRerouteUsage}
  MaterialExpressionNamedRerouteUsage_3 {class=MaterialExpressionNamedRerouteUsage}
  MaterialExpressionNamedRerouteUsage_39 {class=MaterialExpressionNamedRerouteUsage}
  MaterialExpressionNamedRerouteUsage_4 {class=MaterialExpressionNamedRerouteUsage}
  MaterialExpressionNamedRerouteUsage_41 {class=MaterialExpressionNamedRerouteUsage}
  MaterialExpressionNamedRerouteUsage_49 {class=MaterialExpressionNamedRerouteUsage}
  MaterialExpressionNamedRerouteUsage_50 {class=MaterialExpressionNamedRerouteUsage}
  MaterialExpressionNamedRerouteUsage_53 {class=MaterialExpressionNamedRerouteUsage}
  MaterialExpressionOneMinus_3 {class=MaterialExpressionOneMinus, inputs [None {from=MaterialExpressionConvert_6}]}
  MaterialExpressionOneMinus_4 {class=MaterialExpressionOneMinus, inputs [None {from=MaterialExpressionConvert_6}]}
  MaterialExpressionOneMinus_5 {class=MaterialExpressionOneMinus, inputs [None {from=MaterialExpressionMultiply_11}]}
  MaterialExpressionOneMinus_6 {class=MaterialExpressionOneMinus, inputs [None {from=MaterialExpressionNamedRerouteUsage_50}]}
  MaterialExpressionOneMinus_7 {class=MaterialExpressionOneMinus, inputs [None {from=MaterialExpressionNamedRerouteUsage_1}]}
  MaterialExpressionSaturate_0 {class=MaterialExpressionSaturate, inputs [None {from=MaterialExpressionScalarParameter_0}]}
  MaterialExpressionSaturate_6 {class=MaterialExpressionSaturate, inputs [None {from=MaterialExpressionLinearInterpolate_0}]}
  MaterialExpressionSaturate_7 {class=MaterialExpressionSaturate, inputs [None {from=MaterialExpressionScalarParameter_2}]}
  MaterialExpressionScalarParameter_0 {
   class=MaterialExpressionScalarParameter
   properties {group=Emote | Mouth, parameterName="Mouth Use Anim Curve <-> Anim Debug Param", sliderMax=1, sortPriority=10}
  }
  MaterialExpressionScalarParameter_2 {
   class=MaterialExpressionScalarParameter
   properties {group=Emote | Mouth, parameterName=Mouth Inverse Emission, sliderMax=1, sortPriority=60}
  }
  MaterialExpressionScalarParameter_34 {class=MaterialExpressionScalarParameter, properties {group=Emote | Mouth, parameterName=Mouth Anim Debug, sliderMax=1, sortPriority=20}}
  MaterialExpressionScalarParameter_35 {
   class=MaterialExpressionScalarParameter
   properties {defaultValue=0.1, group=Emote | Mouth, parameterName=Mouth Emission Edge Softness, sliderMax=0.5, sortPriority=50}
  }
  MaterialExpressionScalarParameter_36 {
   class=MaterialExpressionScalarParameter
   properties {defaultValue=0.2, group=Emote | Mouth, parameterName=Mouth Off Emission Gain, sliderMax=1, sortPriority=40}
  }
  MaterialExpressionSmoothStep_4 {
   class=MaterialExpressionSmoothStep
   inputs [
    Min {from=MaterialExpressionSubtract_5, output=""}
    Max {from=MaterialExpressionNamedRerouteUsage_53, output=""}
    Value {from=MaterialExpressionComponentMask_3, output=""}
   ]
   properties {desc="Animated Gradient [0, 1]"}
  }
  MaterialExpressionStep_0 {
   class=MaterialExpressionStep
   inputs [
    Y {from="", output=""}
    X {from=MaterialExpressionNamedRerouteUsage_0, output=""}
   ]
   properties {constY=1}
  }
  MaterialExpressionSubtract_1 {
   class=MaterialExpressionSubtract
   inputs [
    A {from=MaterialExpressionTextureCoordinate_4}
    B {from=MaterialExpressionNamedRerouteUsage_39}
   ]
  }
  MaterialExpressionSubtract_3 {
   class=MaterialExpressionSubtract
   inputs [
    A {from=MaterialExpressionNamedRerouteUsage_41}
    B {from=MaterialExpressionNamedRerouteUsage_39}
   ]
  }
  MaterialExpressionSubtract_5 {
   class=MaterialExpressionSubtract
   inputs [
    A {from=MaterialExpressionNamedRerouteUsage_4}
    B {from=MaterialExpressionNamedRerouteUsage_49}
   ]
  }
  MaterialExpressionSwitch_1 {
   class=MaterialExpressionSwitch
   inputs [
    SwitchValue {from=MaterialExpressionNamedRerouteUsage_27, output=""}
    Default {from="", output=""}
    None {from=MaterialExpressionNamedRerouteUsage_1, output=""}
    None {from=MaterialExpressionOneMinus_7, output=""}
   ]
   properties {inputs [
     {
      inputName=None
      input {
       expression=&self:MaterialExpressionNamedRerouteUsage_1
       inputName=None
       mask=0
       maskA=0
       maskB=0
       maskG=0
       maskR=0
       outputIndex=0
      }
     }
     {
      inputName=None
      input {
       expression=&self:MaterialExpressionOneMinus_7
       inputName=None
       mask=0
       maskA=0
       maskB=0
       maskG=0
       maskR=0
       outputIndex=0
      }
     }
    ]}
  }
  MaterialExpressionSwitch_2 {
   class=MaterialExpressionSwitch
   inputs [
    SwitchValue {from=MaterialExpressionNamedRerouteUsage_29, output=""}
    Default {from="", output=""}
    None {from=MaterialExpressionOneMinus_5, output=""}
    None {from=MaterialExpressionMultiply_11, output=""}
   ]
   properties {inputs [
     {
      inputName=None
      input {
       expression=&self:MaterialExpressionOneMinus_5
       inputName=None
       mask=0
       maskA=0
       maskB=0
       maskG=0
       maskR=0
       outputIndex=0
      }
     }
     {
      inputName=None
      input {
       expression=&self:MaterialExpressionMultiply_11
       inputName=None
       mask=0
       maskA=0
       maskB=0
       maskG=0
       maskR=0
       outputIndex=0
      }
     }
    ]}
  }
  MaterialExpressionTextureCoordinate_4 {class=MaterialExpressionTextureCoordinate}
  MaterialExpressionVectorParameter_3 {
   class=MaterialExpressionVectorParameter
   properties {
    group=Emote | Mouth
    parameterName=Mouth UV Mask
    sortPriority=30
    channelNames {a=Max V, b=Max U, g=Min V, r=Min U}
    defaultValue=<rgba 0.53 0.73 0.92 0.95>
   }
  }
 }}
