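# Character body material for the DandyFlare cosmetic skin, built between the shared 2023 CharacterParent PreFX and PostFX functions.
# Static switches select a mouth emissive animation, else a disc one, else a neutral default, yielding an emissive gain and area mask.
# The mouth effect follows the jaw_open_pose parameter remapped by SmoothStep.
# The PreFX emissive gets multiplied, then passes through DandyFlare's AdvancedEmissive function with a masked alpha before PostFX.
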Material /BRCosmetics/Characters/Player/Male/Large/Bodies/M_LRG_DandyFlare/Materials/M_DandyFlare
graph {
 expressions {
  MaterialExpressionConvert_0 {
   class=MaterialExpressionConvert
   inputs [None {from=MaterialExpressionStaticSwitchParameter_4}]
   properties {
    convertInputs [{
      type=Vector4
      defaultValue=<rgba 0 0 0 1>
      expressionInput {
       expression=&self:MaterialExpressionStaticSwitchParameter_4
       inputName=None
       mask=0
       maskA=0
       maskB=0
       maskG=0
       maskR=0
       outputIndex=0
      }
     }]
    convertOutputs [{type=Vector3}, {type=Scalar}]
   }
  }
  MaterialExpressionConvert_1 {class=MaterialExpressionConvert, inputs [], properties {convertOutputs [{type=Vector4, defaultValue=<rgba 1 1 1 0>}]}}
  MaterialExpressionConvert_3 {
   class=MaterialExpressionConvert
   inputs [
    None {from=MaterialExpressionMaterialFunctionCall_3, output="Emission Gain (V3)"}
    None {from=MaterialExpressionMaterialFunctionCall_3, output="Emission Gain (V3)"}
   ]
   properties {
    convertInputs [
     {
      type=Vector3
      expressionInput {
       expression=&self:MaterialExpressionMaterialFunctionCall_3
       inputName=None
       mask=0
       maskA=0
       maskB=0
       maskG=0
       maskR=0
       outputIndex=0
      }
     }
     {
      type=Scalar
      expressionInput {
       expression=&self:MaterialExpressionMaterialFunctionCall_3
       inputName=None
       mask=0
       maskA=0
       maskB=0
       maskG=0
       maskR=0
       outputIndex=1
      }
     }
    ]
    convertOutputs [{type=Vector4, defaultValue=<rgba 0 0 0 1>}]
   }
  }
  MaterialExpressionConvert_7 {
   class=MaterialExpressionConvert
   inputs [
    None {from=MaterialExpressionMaterialFunctionCall_4, output="Emissive Gain (S)"}
    None {from=MaterialExpressionMaterialFunctionCall_4, output="Emissive Gain (S)"}
   ]
   properties {
    convertInputs [
     {
      type=Scalar
      expressionInput {
       expression=&self:MaterialExpressionMaterialFunctionCall_4
       inputName=None
       mask=0
       maskA=0
       maskB=0
       maskG=0
       maskR=0
       outputIndex=0
      }
     }
     {
      type=Scalar
      expressionInput {
       expression=&self:MaterialExpressionMaterialFunctionCall_4
       inputName=None
       mask=0
       maskA=0
       maskB=0
       maskG=0
       maskR=0
       outputIndex=1
      }
     }
    ]
    convertOutputs [{type=Vector4, defaultValue=<rgba 0 0 0 1>}]
   }
  }
  MaterialExpressionGetMaterialAttributes_1 {
   class=MaterialExpressionGetMaterialAttributes
   inputs [None {from=MaterialExpressionMaterialFunctionCall_0, output=Result}]
   properties {attributeGetTypes [#1]}
  }
  MaterialExpressionMaterialFunctionCall_0 {
   class=MaterialExpressionMaterialFunctionCall
   inputs [
    Overide Texture Set UVs {from=""}
    Override Tech Art Mask UVs {from=""}
    Override Basic Emissive UVs {from=""}
   ]
   properties {materialFunction=&MF_CharacterParent_2023_PreFX}
  }
  MaterialExpressionMaterialFunctionCall_1 {
   class=MaterialExpressionMaterialFunctionCall
   inputs [
    Material {from=MaterialExpressionMaterialFunctionCall_10, output=Result}
    M {from=MaterialExpressionMaterialFunctionCall_0, output=M}
    Pre Skinned Local Position {from=MaterialExpressionMaterialFunctionCall_0, output=M}
    Distance Blend {from=MaterialExpressionMaterialFunctionCall_0, output=M}
    Fresnel {from=MaterialExpressionMaterialFunctionCall_0, output=M}
   ]
   properties {materialFunction=&MF_CharacterParent_2023_PostFX}
  }
  MaterialExpressionMaterialFunctionCall_10 {
   class=MaterialExpressionMaterialFunctionCall
   inputs [
    Local Position {from=MaterialExpressionNamedRerouteUsage_33}
    Alpha {from=MaterialExpressionMultiply_14}
    Fresnel {from=MaterialExpressionNamedRerouteUsage_32}
    Material {from=MaterialExpressionSetMaterialAttributes_1}
    Local Normal {from=""}
   ]
   properties {materialFunction=/BRCosmetics/Characters/Player/Male/Large/Bodies/M_LRG_DandyFlare/Materials/MaterialFunctions/MF_DandyFlare_AdvancedEmissive}
  }
  MaterialExpressionMaterialFunctionCall_3 {
   class=MaterialExpressionMaterialFunctionCall
   inputs []
   properties {materialFunction=/BRCosmetics/Characters/Player/Male/Large/Bodies/M_LRG_DandyFlare/Materials/MaterialFunctions/MF_DandyFlare_Disc_Emissive}
  }
  MaterialExpressionMaterialFunctionCall_4 {
   class=MaterialExpressionMaterialFunctionCall
   inputs [Anim Curve {from=MaterialExpressionNamedRerouteDeclaration_44}]
   properties {materialFunction=/BRCosmetics/Characters/Player/Male/Large/Bodies/M_LRG_DandyFlare/Materials/MaterialFunctions/MF_DandyFlare_Mouth_Emissive}
  }
  MaterialExpressionMultiply_13 {
   class=MaterialExpressionMultiply
   inputs [
    A {from=MaterialExpressionNamedRerouteUsage_31}
    B {from=MaterialExpressionReroute_1}
   ]
  }
  MaterialExpressionMultiply_14 {
   class=MaterialExpressionMultiply
   inputs [
    A {from=MaterialExpressionNamedRerouteUsage_0}
    B {from=MaterialExpressionOneMinus_8}
   ]
  }
  MaterialExpressionNamedRerouteDeclaration_0 {
   class=MaterialExpressionNamedRerouteDeclaration
   inputs [None {from=MaterialExpressionMaterialFunctionCall_0, output=Tech Art Mask}]
   properties {name="TechArtMask (V3)", nodeColor=<rgba 0.5913 0 1 1>}
  }
  MaterialExpressionNamedRerouteDeclaration_15 {
   class=MaterialExpressionNamedRerouteDeclaration
   inputs [None {from=MaterialExpressionConvert_0}]
   properties {name="Emote Emissive Gain (V3)", nodeColor=<rgba 1 0.75 0 1>}
  }
  MaterialExpressionNamedRerouteDeclaration_17 {
   class=MaterialExpressionNamedRerouteDeclaration
   inputs [None {from=MaterialExpressionMaterialFunctionCall_0, output=Pre skinned Local Position}]
   properties {name="Local Position (V3)", nodeColor=<rgba 1 0.1043 0 1>}
  }
  MaterialExpressionNamedRerouteDeclaration_18 {
   class=MaterialExpressionNamedRerouteDeclaration
   inputs [None {from=MaterialExpressionMaterialFunctionCall_0, output=Fresnel}]
   properties {name="Fresnel (S)", nodeColor=<rgba 0 0.7131 1 1>}
  }
  MaterialExpressionNamedRerouteDeclaration_44 {
   class=MaterialExpressionNamedRerouteDeclaration
   inputs [None {from=MaterialExpressionSmoothStep_0}]
   properties {name="Anim Curve (S)", nodeColor=<rgba 0 1 0.6953 1>}
  }
  MaterialExpressionNamedRerouteDeclaration_46 {
   class=MaterialExpressionNamedRerouteDeclaration
   inputs [None {from=MaterialExpressionConvert_0}]
   properties {name="Emote Area Mask (S)", nodeColor=<rgba 0.1016 1 0 1>}
  }
  MaterialExpressionNamedRerouteUsage_0 {class=MaterialExpressionNamedRerouteUsage}
  MaterialExpressionNamedRerouteUsage_31 {class=MaterialExpressionNamedRerouteUsage}
  MaterialExpressionNamedRerouteUsage_32 {class=MaterialExpressionNamedRerouteUsage}
  MaterialExpressionNamedRerouteUsage_33 {class=MaterialExpressionNamedRerouteUsage}
  MaterialExpressionNamedRerouteUsage_57 {class=MaterialExpressionNamedRerouteUsage}
  MaterialExpressionOneMinus_8 {class=MaterialExpressionOneMinus, inputs [None {from=MaterialExpressionNamedRerouteUsage_57}]}
  MaterialExpressionReroute_1 {class=MaterialExpressionReroute, inputs [None {from=MaterialExpressionGetMaterialAttributes_1, output=Emissive Color}]}
  MaterialExpressionScalarParameter_1 {class=MaterialExpressionScalarParameter, properties {group=Emote | Mouth, parameterName=jaw_open_pose, sortPriority=-90}}
  MaterialExpressionScalarParameter_3 {class=MaterialExpressionScalarParameter, properties {group=Emote | Mouth, parameterName=Anim Curve Min, sliderMax=1, sortPriority=-80}}
  MaterialExpressionScalarParameter_6 {
   class=MaterialExpressionScalarParameter
   properties {defaultValue=1, group=Emote | Mouth, parameterName=Anim Curve Max, sliderMax=1, sortPriority=-70}
  }
  MaterialExpressionSetMaterialAttributes_1 {
   class=MaterialExpressionSetMaterialAttributes
   inputs [
    MaterialAttributes {from=MaterialExpressionGetMaterialAttributes_1, output=MaterialAttributes}
    Emissive Color {from=MaterialExpressionMultiply_13, output=""}
   ]
   properties {attributeSetTypes [#1]}
  }
  MaterialExpressionSmoothStep_0 {
   class=MaterialExpressionSmoothStep
   inputs [
    Min {from=MaterialExpressionScalarParameter_3, output=""}
    Max {from=MaterialExpressionScalarParameter_6, output=""}
    Value {from=MaterialExpressionScalarParameter_1, output=""}
   ]
   properties {}
  }
  MaterialExpressionStaticSwitchParameter_3 {
   class=MaterialExpressionStaticSwitchParameter
   inputs [
    True {from=MaterialExpressionConvert_3, output=""}
    False {from=MaterialExpressionConvert_1, output=""}
   ]
   properties {group=Emote | Disc, parameterName=Emote - Enable Disc Emissive Anim, sortPriority=-10}
  }
  MaterialExpressionStaticSwitchParameter_4 {
   class=MaterialExpressionStaticSwitchParameter
   inputs [
    True {from=MaterialExpressionConvert_7, output=""}
    False {from=MaterialExpressionStaticSwitchParameter_3, output=""}
   ]
   properties {group=Emote | Mouth, parameterName=Emote - Enable Mouth Emissive Anim, sortPriority=-100}
  }
 }
 properties {
  bAutomaticallySetUsageInEditor=false
  bUseMaterialAttributes=true
  bUsedWithMeshDeformer=true
  bUsedWithMorphTargets=true
  bUsedWithSkeletalMesh=true
  pixelDepthOffsetMode=PDOM_Legacy
  shadingModel=MSM_FromMaterialExpression
  usedShadingModels=Default Lit | Subsurface | Clear Coat | Subsurface Profile | Cloth
 }
}
