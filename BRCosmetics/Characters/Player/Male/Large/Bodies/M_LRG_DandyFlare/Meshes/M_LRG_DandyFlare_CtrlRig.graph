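# Control Rig for the large male DandyFlare player body, driving its deform bones as corrective post-animation deformation.
# A shared deform-rig function and arm retargeting with constraints and full-body IK run, with elbow-to-head distance weighting the corrections.
# Spherical pose readers on neck, thighs and upper arms weight collar, thigh, shoulder and chest offsets.
# Hand rotation drives wrist bones, and shoulder correctives switch when skydiving.
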ControlRigBlueprint /BRCosmetics/Characters/Player/Male/Large/Bodies/M_LRG_DandyFlare/Meshes/M_LRG_DandyFlare_CtrlRig
graphs [
 New Function_ContainedGraph {nodes [
   RigVMFunctionEntryNode()
   > RigVMFunction_Sequence = RigVMAggregateNode()
   &RigUnit_AimConstraintLocalSpaceOffset(
    Child=(Type=Bone,Name="deform_upperarm_l")
    Filter=(bX=True,bY=True,bZ=True)
    Parents=((Item=(Type=Bone,Name="lowerarm_l"),Weight=1))
    Weight=1
    WorldUp=(Target=(X=0,Y=0,Z=1),Kind=Location,Space=(Type=Bone,Name="lowerarm_l"))
    ExecutePin=@RigVMFunction_Sequence.B
   )
   > &RigUnit_ParentConstraint(
    AdvancedSettings=(InterpolationType=Average,RotationOrderForFilter=XZY)
    Child=(Type=Bone,Name="deform_lowerarm_l")
    Filter=(TranslationFilter=(bX=false,bY=false,bZ=false),RotationFilter=(bX=True,bY=True,bZ=True),ScaleFilter=(bX=True,bY=True,bZ=True))
    Parents=((Item=(Type=Bone,Name="lowerarm_l"),Weight=1))
    Weight=1
   )
   > &RigUnit_ParentConstraint(
    AdvancedSettings=(InterpolationType=Average,RotationOrderForFilter=XZY)
    Child=(Type=Bone,Name="deform_hand_l")
    Parents=((Item=(Type=Bone,Name="hand_l"),Weight=1))
    Weight=1
   )
   > &RigUnit_AimConstraintLocalSpaceOffset(
    Child=(Type=Bone,Name="deform_upperarm_r")
    Filter=(bX=True,bY=True,bZ=True)
    Parents=((Item=(Type=Bone,Name="cr_lowerarm_push_r"),Weight=1))
    Weight=1
    WorldUp=(Target=(X=0,Y=0,Z=1),Kind=Location,Space=(Type=Bone,Name="cr_lowerarm_push_r"))
   )
   > &RigUnit_ParentConstraint(
    AdvancedSettings=(InterpolationType=Average,RotationOrderForFilter=XZY)
    Child=(Type=Bone,Name="deform_lowerarm_r")
    Parents=((Item=(Type=Bone,Name="cr_lowerarm_push_r"),Weight=1))
    Weight=1
   )
   > &RigUnit_ParentConstraint(
    AdvancedSettings=(InterpolationType=Average,RotationOrderForFilter=XZY)
    Child=(Type=Bone,Name="deform_hand_r")
    Parents=((Item=(Type=Bone,Name="hand_r"),Weight=1))
    Weight=1
   )
   > /Script/ControlRig.RigUnit_PositionConstraintLocalSpaceOffset(
    Child=(Type=Bone,Name="deform_upperarm_l_end")
    Filter=(bX=True,bY=True,bZ=True)
    Parents=((Item=(Type=Bone,Name="deform_lowerarm_l"),Weight=1))
    Weight=0.8
    bMaintainOffset=True
   ) {is=struct RigVMUnitNode}
   > &RigUnit_PositionConstraintLocalSpaceOffset(
    Child=(Type=Bone,Name="deform_upperarm_r_end")
    Filter=(bX=True,bY=True,bZ=True)
    Parents=((Item=(Type=Bone,Name="deform_lowerarm_r"),Weight=1))
    Weight=0.8
    bMaintainOffset=True
   )
   /Script/PBIK.RigUnit_PBIK(
    BoneSettings=((Bone="deform_hand_l",RotationStiffness=0,PositionStiffness=0,X=Free,MinX=0,MaxX=0,Y=Free,MinY=0,MaxY=0,Z=Free,MinZ=0,MaxZ=0,bUsePreferredAngles=true,PreferredAngles=(X=1,Y=0,Z=0)))
    Debug=(DrawScale=1,bDrawDebug=false)
    Effectors=((Bone="hand_l",Transform=(Rotation=(X=0,Y=0,Z=0,W=1),Translation=(X=0,Y=0,Z=0),Scale3D=(X=1,Y=1,Z=1)),PositionAlpha=1,RotationAlpha=1,StrengthAlpha=1,PullChainAlpha=1,PinRotation=1,ChainDepth=0))
    ExcludedBones=()
    Root=deform_upperarm_l
    Settings="(Iterations=20,SubIterations=0,MassMultiplier=1,bAllowStretch=false,RootBehavior=PrePull,PrePullRootSettings=(RotationAlpha=0,RotationAlphaX=1,RotationAlphaY=1,RotationAlphaZ=1,PositionAlpha=1,PositionAlphaX=1,PositionAlphaY=1,PositionAlphaZ=1),GlobalPullC ...(truncated)"
    ExecutePin=@RigVMFunction_Sequence.C
   ) {is=struct RigVMUnitNode}
   > &RigUnit_PBIK(
    BoneSettings="((Bone=\"deform_hand_r\",RotationStiffness=0,PositionStiffness=0,X=Free,MinX=0,MaxX=0,Y=Free,MinY=0,MaxY=0,Z=Free,MinZ=0,MaxZ=0,bUsePreferredAngles=false,PreferredAngles=(X=1,Y=0,Z=0)),(Bone=\"deform_upperarm_l\",RotationStiffness=0,PositionStiffness=0,X=Free, ...(truncated)"
    Debug=(DrawScale=3,bDrawDebug=true)
    Effectors="((Bone=\"hand_r\",Transform=(Rotation=(X=0,Y=0,Z=0,W=1),Translation=(X=0,Y=0,Z=0),Scale3D=(X=1,Y=1,Z=1)),PositionAlpha=1,RotationAlpha=1,StrengthAlpha=1,PullChainAlpha=1,PinRotation=1,ChainDepth=0),(Bone=\"upperarm_r\",Transform=(Rotation=(X=0,Y=0,Z=0,W=1),Tra ...(truncated)"
    ExcludedBones=()
    Root=deform_upperarm_r
    Settings="(Iterations=20,SubIterations=0,MassMultiplier=1,bAllowStretch=true,RootBehavior=PrePull,PrePullRootSettings=(RotationAlpha=0,RotationAlphaX=1,RotationAlphaY=1,RotationAlphaZ=1,PositionAlpha=1,PositionAlphaX=1,PositionAlphaY=1,PositionAlphaZ=1),GlobalPullCh ...(truncated)"
   )
   > ApplyFBIKChainCorrection_1_3 = RigVMFunctionReferenceNode(
    End=(Type=Bone,Name="deform_hand_r")
    QuatAmount=2.5
    Start=(Type=Bone,Name="deform_upperarm_r")
    Weight=RigVMVariableNode() {variable=R Elbow Distance to Head}.Value
   )
   > ApplyFBIKChainCorrection_1_4 = RigVMFunctionReferenceNode(
    End=(Type=Bone,Name="deform_hand_l")
    QuatAmount=2.5
    Start=(Type=Bone,Name="deform_upperarm_l")
    Weight=RigVMVariableNode() {variable=L Elbow Distance to Head}.Value
   )
   RigVMFunctionReturnNode(ExecuteContext=@RigVMFunction_Sequence.D)
   Get Normalized Distance Between_1_1 = RigVMFunctionReferenceNode(
    Debug=true
    Max Distance=35
    Min Distance=1
    Source=(Type=Bone,Name="lowerarm_r")
    Target=(Type=Bone,Name="head")
    ExecuteContext=@RigVMFunction_Sequence.A
   )
   > RigVMVariableNode(Value=@"Get Normalized Distance Between_1_1".Result) {variable=R Elbow Distance to Head}
   > &RigUnit_ModifyTransforms(
    ItemToModify=((Item=(Type=Bone,Name="cr_lowerarm_push_r"),Transform=(Rotation=(X=0,Y=0,Z=0,W=1),Translation=(X=0,Y=90,Z=0),Scale3D=(X=1,Y=1,Z=1))))
    WeightMinimum=0
    Weight=RigVMVariableNode() {variable=R Elbow Distance to Head}.Value
   )
   > &RigUnit_AimConstraintLocalSpaceOffset(
    Child=(Type=Bone,Name="cr_lowerarm_push_r")
    Filter=(bX=True,bY=True,bZ=True)
    Parents=((Item=(Type=Bone,Name="hand_r"),Weight=1))
    Weight=1
    WorldUp=(Target=(X=0,Y=0,Z=1),Kind=Location,Space=(Type=Bone,Name="hand_r"))
   )
   > Get Normalized Distance Between_1_2 = RigVMFunctionReferenceNode(Debug=true, Max Distance=20, Min Distance=1, Source=(Type=Bone,Name="hand_l"), Target=(Type=Bone,Name="head"))
   > RigVMVariableNode(Value=@"Get Normalized Distance Between_1_2".Result) {variable=L Elbow Distance to Head}
   > &RigUnit_ModifyTransforms(
    ItemToModify=((Item=(Type=Bone,Name="cr_lowerarm_push_l"),Transform=(Rotation=(X=0,Y=0,Z=0,W=1),Translation=(X=0,Y=5,Z=0),Scale3D=(X=1,Y=1,Z=1))))
    WeightMinimum=0
    Weight=RigVMVariableNode() {variable=L Elbow Distance to Head}.Value
   )
   > &RigUnit_AimConstraintLocalSpaceOffset(
    Child=(Type=Bone,Name="cr_lowerarm_push_l")
    Filter=(bX=True,bY=True,bZ=True)
    Parents=((Item=(Type=Bone,Name="hand_l"),Weight=1))
    Weight=1
    WorldUp=(Target=(X=0,Y=0,Z=1),Kind=Location,Space=(Type=Bone,Name="hand_l"))
   )
   RigVMVariableNode(ExecuteContext=()) {variable=R Elbow Distance to Head}
   RigVMVariableNode() {variable=R Elbow Distance to Head}
  ]}
 New Function_ContainedGraph {nodes [
   n1 = RigVMFunctionEntryNode()
   n2 = &RigUnit_GetTransform(bInitial=False, Item=@n1.Start)
   Subtract = /Script/RigVM.RigVMFunction_MathVectorSub(B=@n2.Transform.Translation, A=&RigUnit_GetTransform(bInitial=False, Item=@n1.End).Transform.Translation) {is=struct RigVMUnitNode, template="Subtract::Execute(in A,in B,out Result)"}
   Unit = /Script/RigVM.RigVMFunction_MathVectorUnit(Value=@Subtract.Result) {is=struct RigVMUnitNode, template="Unit::Execute(in Value,out Result)"}
   RigVMFunction_MathQuaternionFromAxisAndAngle = /Script/RigVM.RigVMFunction_MathQuaternionFromAxisAndAngle(Axis=@Unit.Result, Angle=@n1.QuatAmount) {is=struct RigVMUnitNode, template="FromAxisAndAngle::Execute(in Axis,in Angle,out Result)"}
   Multiply = /Script/RigVM.RigVMFunction_MathQuaternionMul(B=@n2.Transform.Rotation, A=@RigVMFunction_MathQuaternionFromAxisAndAngle.Result) {is=struct RigVMUnitNode, template="Multiply::Execute(in A,in B,out Result)"}
   > &RigUnit_SetRotation(Space=GlobalSpace, bInitial=False, Item=@n1.Start, Weight=@n1.Weight, Value=@Multiply.Result)
   > RigVMFunctionReturnNode()
  ]}
 New Function_ContainedGraph {nodes [
   RigVMFunctionEntryNode()
   Deform Fingers = RigVMDispatchNode(Value="((Type=Bone,Name=\"deform_thumb_01_l\"),(Type=Bone,Name=\"deform_thumb_02_l\"),(Type=Bone,Name=\"deform_thumb_03_l\"),(Type=Bone,Name=\"deform_index_metacarpal_l\"),(Type=Bone,Name=\"deform_index_01_l\"),(Type=Bone,Name=\"deform_index_02_l\"),(Type=Bone,Name=\"deform_i ...(truncated)") {template="DISPATCH_RigVMDispatch_Constant(io Value)"}
   > For_Each_7 = RigVMDispatchNode(Array=@"Deform Fingers".Value) {template="DISPATCH_RigVMDispatch_ArrayIterator(in Array,out Element,out Index,out Count,out Ratio)"}
   Base Fingers = RigVMDispatchNode(Value="((Type=Bone,Name=\"thumb_01_l\"),(Type=Bone,Name=\"thumb_02_l\"),(Type=Bone,Name=\"thumb_03_l\"),(Type=Bone,Name=\"index_metacarpal_l\"),(Type=Bone,Name=\"index_01_l\"),(Type=Bone,Name=\"index_02_l\"),(Type=Bone,Name=\"index_03_l\"),(Type=Bone,Name=\"middle_metacarpal_l\" ...(truncated)") {template="DISPATCH_RigVMDispatch_Constant(io Value)"}
   > &RigUnit_ParentConstraint(
    AdvancedSettings=(InterpolationType=Average,RotationOrderForFilter=XZY)
    Parents=((Item=(Type=Bone,Name="None"),Weight=1))
    Weight=1
    Parents.0.Item=RigVMDispatchNode(Array=@"Base Fingers".Value, Index=@For_Each_7.Index) {template="DISPATCH_RigVMDispatch_ArrayGetAtIndex(in Array,in Index,out Element)"}.Element
    Child=@For_Each_7.Element
   )
   RigVMFunctionReturnNode(ExecuteContext=@For_Each_7.Completed)
  ]}
 New Function_ContainedGraph {nodes [
   RigVMFunctionEntryNode()
   > &RigUnit_ParentConstraint(
    AdvancedSettings=(InterpolationType=Average,RotationOrderForFilter=XZY)
    Child=(Type=Bone,Name="deform_elbow_in_l_base")
    Parents=((Item=(Type=Bone,Name="deform_lowerarm_l"),Weight=1),(Item=(Type=Bone,Name="deform_upperarm_l_end"),Weight=1))
    Weight=1
   )
   > &RigUnit_ParentConstraint(
    AdvancedSettings=(InterpolationType=Average,RotationOrderForFilter=XZY)
    Child=(Type=Bone,Name="deform_elbow_in_r_base")
    Parents=((Item=(Type=Bone,Name="deform_lowerarm_r"),Weight=1),(Item=(Type=Bone,Name="deform_upperarm_r_end"),Weight=1))
    Weight=1
   )
   > Get Delta Ref = RigVMFunctionReferenceNode(Item=(Type=Bone,Name="deform_lowerarm_l"))
   > Get Delta Ref_1 = RigVMFunctionReferenceNode(Item=(Type=Bone,Name="deform_lowerarm_r"))
   > Modify Transforms with Mirror = RigVMFunctionReferenceNode(
    Base Modify Data=((Item=(Type=Bone,Name="deform_elbow_in_l"),Transform=(Rotation=(X=0,Y=0,Z=-0.1305,W=0.9914),Translation=(X=1,Y=4,Z=0),Scale3D=(X=1,Y=1,Z=1))))
    Mirror Modify Data=((Item=(Type=Bone,Name="deform_elbow_in_r"),Transform=(Rotation=(X=0,Y=0,Z=0,W=1),Translation=(X=0,Y=0,Z=0),Scale3D=(X=1,Y=1,Z=1))))
    Weight=&RigVMFunction_MathFloatRemap(SourceMaximum=90, SourceMinimum=15, bClamp=true, Value=@"Get Delta Ref".Result.Rotation.Yaw).Result
    "Mirror Weight "=&RigVMFunction_MathFloatRemap(SourceMaximum=90, SourceMinimum=15, bClamp=true, Value=@"Get Delta Ref_1".Result.Rotation.Yaw).Result
   )
   > RigVMFunctionReturnNode()
  ]}
 New Function_ContainedGraph {nodes [
   RigVMFunctionEntryNode()
   > RigVMFunction_Sequence = RigVMAggregateNode()
   &RigUnit_AimConstraintLocalSpaceOffset(
    Child=(Type=Bone,Name="deform_upperarm_l")
    Filter=(bX=True,bY=True,bZ=True)
    Parents=((Item=(Type=Bone,Name="lowerarm_l"),Weight=1))
    Weight=1
    WorldUp=(Target=(X=0,Y=0,Z=1),Kind=Location,Space=(Type=Bone,Name="lowerarm_l"))
    bMaintainOffset=true
    ExecutePin=@RigVMFunction_Sequence.B
   )
   > &RigUnit_ParentConstraint(
    AdvancedSettings=(InterpolationType=Average,RotationOrderForFilter=XZY)
    Child=(Type=Bone,Name="deform_lowerarm_l")
    Filter=(TranslationFilter=(bX=true,bY=true,bZ=true),RotationFilter=(bX=True,bY=True,bZ=True),ScaleFilter=(bX=True,bY=True,bZ=True))
    Parents=((Item=(Type=Bone,Name="lowerarm_l"),Weight=1))
    Weight=1
    bMaintainOffset=true
   )
   > &RigUnit_ParentConstraint(
    AdvancedSettings=(InterpolationType=Average,RotationOrderForFilter=XZY)
    Child=(Type=Bone,Name="deform_hand_l")
    Filter=(TranslationFilter=(bX=true,bY=true,bZ=true),RotationFilter=(bX=True,bY=True,bZ=True),ScaleFilter=(bX=True,bY=True,bZ=True))
    Parents=((Item=(Type=Bone,Name="hand_l"),Weight=1))
    Weight=1
    bMaintainOffset=true
   )
   > &RigUnit_AimConstraintLocalSpaceOffset(
    Child=(Type=Bone,Name="deform_upperarm_r")
    Filter=(bX=True,bY=True,bZ=True)
    Parents=((Item=(Type=Bone,Name="cr_lowerarm_push_r"),Weight=1))
    Weight=1
    WorldUp=(Target=(X=0,Y=0,Z=1),Kind=Location,Space=(Type=Bone,Name="cr_lowerarm_push_r"))
   )
   > &RigUnit_ParentConstraint(
    AdvancedSettings=(InterpolationType=Average,RotationOrderForFilter=XZY)
    Child=(Type=Bone,Name="deform_lowerarm_r")
    Filter=(TranslationFilter=(bX=true,bY=true,bZ=true),RotationFilter=(bX=True,bY=True,bZ=True),ScaleFilter=(bX=True,bY=True,bZ=True))
    Parents=((Item=(Type=Bone,Name="lowerarm_r"),Weight=1))
    Weight=1
   )
   > &RigUnit_ParentConstraint(
    AdvancedSettings=(InterpolationType=Average,RotationOrderForFilter=XZY)
    Child=(Type=Bone,Name="deform_hand_r")
    Parents=((Item=(Type=Bone,Name="hand_r"),Weight=1))
    Weight=1
    bMaintainOffset=true
   )
   > &RigUnit_PositionConstraintLocalSpaceOffset(
    Child=(Type=Bone,Name="deform_upperarm_l_end")
    Filter=(bX=True,bY=True,bZ=True)
    Parents=((Item=(Type=Bone,Name="deform_lowerarm_l"),Weight=1))
    Weight=0.7
    bMaintainOffset=true
   )
   > &RigUnit_PositionConstraintLocalSpaceOffset(
    Child=(Type=Bone,Name="deform_upperarm_r_end")
    Filter=(bX=True,bY=True,bZ=True)
    Parents=((Item=(Type=Bone,Name="deform_lowerarm_r"),Weight=1))
    Weight=0.7
    bMaintainOffset=true
   )
   &RigUnit_PBIK(
    BoneSettings="((Bone=\"deform_hand_l\",RotationStiffness=0,PositionStiffness=0,X=Free,MinX=0,MaxX=0,Y=Free,MinY=0,MaxY=0,Z=Free,MinZ=0,MaxZ=0,bUsePreferredAngles=true,PreferredAngles=(X=1,Y=0,Z=0)),(Bone=\"deform_upperarm_l\",RotationStiffness=0,PositionStiffness=0,X=Free,M ...(truncated)"
    Debug=(DrawScale=1,bDrawDebug=false)
    Effectors="((Bone=\"hand_l\",Transform=(Rotation=(X=0,Y=0,Z=0,W=1),Translation=(X=0,Y=0,Z=0),Scale3D=(X=1,Y=1,Z=1)),PositionAlpha=1,RotationAlpha=1,StrengthAlpha=1,PullChainAlpha=1,PinRotation=1,ChainDepth=0),(Bone=\"upperarm_l\",Transform=(Rotation=(X=0,Y=0,Z=0,W=1),Tra ...(truncated)"
    ExcludedBones=()
    Root=deform_upperarm_l
    Settings="(Iterations=20,SubIterations=0,MassMultiplier=1,bAllowStretch=false,RootBehavior=PrePull,PrePullRootSettings=(RotationAlpha=0,RotationAlphaX=1,RotationAlphaY=1,RotationAlphaZ=1,PositionAlpha=1,PositionAlphaX=1,PositionAlphaY=1,PositionAlphaZ=1),GlobalPullC ...(truncated)"
    ExecutePin=@RigVMFunction_Sequence.C
   )
   > &RigUnit_PBIK(
    BoneSettings="((Bone=\"deform_hand_r\",RotationStiffness=0,PositionStiffness=0,X=Free,MinX=0,MaxX=0,Y=Free,MinY=0,MaxY=0,Z=Free,MinZ=0,MaxZ=0,bUsePreferredAngles=false,PreferredAngles=(X=1,Y=0,Z=0)),(Bone=\"deform_upperarm_l\",RotationStiffness=0,PositionStiffness=0,X=Free, ...(truncated)"
    Debug=(DrawScale=3,bDrawDebug=true)
    Effectors="((Bone=\"hand_r\",Transform=(Rotation=(X=0,Y=0,Z=0,W=1),Translation=(X=0,Y=0,Z=0),Scale3D=(X=1,Y=1,Z=1)),PositionAlpha=1,RotationAlpha=1,StrengthAlpha=1,PullChainAlpha=1,PinRotation=1,ChainDepth=0),(Bone=\"upperarm_r\",Transform=(Rotation=(X=0,Y=0,Z=0,W=1),Tra ...(truncated)"
    ExcludedBones=()
    Root=deform_upperarm_r
    Settings="(Iterations=20,SubIterations=0,MassMultiplier=1,bAllowStretch=true,RootBehavior=PrePull,PrePullRootSettings=(RotationAlpha=0,RotationAlphaX=1,RotationAlphaY=1,RotationAlphaZ=1,PositionAlpha=1,PositionAlphaX=1,PositionAlphaY=1,PositionAlphaZ=1),GlobalPullCh ...(truncated)"
   )
   Multiply = &RigVMFunction_MathDoubleMul(
    A=RigVMVariableNode() {variable=R Elbow Distance to Head}.Value
    B=RigVMVariableNode() {variable=RightArm_is_HidenIsWearingGauntlet_alpha_R}.Value
   )
   > ApplyFBIKChainCorrection_1_1 = RigVMFunctionReferenceNode(End=(Type=Bone,Name="deform_hand_r"), QuatAmount=2.5, Start=(Type=Bone,Name="deform_upperarm_r"), Weight=@Multiply.Result)
   > ApplyFBIKChainCorrection_1_2 = RigVMFunctionReferenceNode(
    End=(Type=Bone,Name="deform_hand_l")
    QuatAmount=2.5
    Start=(Type=Bone,Name="deform_upperarm_l")
    Weight=RigVMVariableNode() {variable=L Elbow Distance to Head}.Value
   )
   Get Normalized Distance Between_1_1 = RigVMFunctionReferenceNode(
    Debug=false
    Max Distance=35
    Min Distance=1
    Source=(Type=Bone,Name="lowerarm_r")
    Target=(Type=Bone,Name="head")
    ExecuteContext=@RigVMFunction_Sequence.A
   )
   > RigVMVariableNode(Value=@"Get Normalized Distance Between_1_1".Result) {variable=R Elbow Distance to Head}
   > &RigUnit_ModifyTransforms(
    ItemToModify=((Item=(Type=Bone,Name="cr_lowerarm_push_r"),Transform=(Rotation=(X=0,Y=0,Z=0,W=1),Translation=(X=0,Y=4.5,Z=-5),Scale3D=(X=1,Y=1,Z=1))))
    WeightMinimum=0
    Weight=RigVMVariableNode() {variable=R Elbow Distance to Head}.Value
   )
   > &RigUnit_AimConstraintLocalSpaceOffset(
    Child=(Type=Bone,Name="cr_lowerarm_push_r")
    Filter=(bX=True,bY=True,bZ=True)
    Parents=((Item=(Type=Bone,Name="hand_r"),Weight=1))
    Weight=1
    WorldUp=(Target=(X=0,Y=0,Z=1),Kind=Location,Space=(Type=Bone,Name="hand_r"))
   )
   > Get Normalized Distance Between_1_2 = RigVMFunctionReferenceNode(Debug=false, Max Distance=35, Min Distance=1, Source=(Type=Bone,Name="lowerarm_l"), Target=(Type=Bone,Name="head"))
   > RigVMVariableNode(Value=@"Get Normalized Distance Between_1_2".Result) {variable=L Elbow Distance to Head}
   > &RigUnit_ModifyTransforms(
    ItemToModify=((Item=(Type=Bone,Name="cr_lowerarm_push_l"),Transform=(Rotation=(X=0,Y=0,Z=0,W=1),Translation=(X=0,Y=5,Z=0),Scale3D=(X=1,Y=1,Z=1))))
    WeightMinimum=0
    Weight=RigVMVariableNode() {variable=L Elbow Distance to Head}.Value
   )
   > &RigUnit_AimConstraintLocalSpaceOffset(
    Child=(Type=Bone,Name="cr_lowerarm_push_l")
    Filter=(bX=True,bY=True,bZ=True)
    Parents=((Item=(Type=Bone,Name="hand_l"),Weight=1))
    Weight=1
    WorldUp=(Target=(X=0,Y=0,Z=1),Kind=Location,Space=(Type=Bone,Name="hand_l"))
   )
   RigVMFunctionReturnNode(ExecuteContext=@RigVMFunction_Sequence.D)
  ]}
 New Function_ContainedGraph {nodes [
   RigVMFunctionEntryNode()
   Multiply = &RigVMFunction_MathTransformMul(
    B=/Script/RigVM.RigVMFunction_MathTransformInverse(Value=&RigUnit_GetTransform(Item=(Type=Bone,Name="spine_05"), Space=LocalSpace, bInitial=true).Transform) {is=struct RigVMUnitNode, template="Inverse::Execute(in Value,out Result)"}.Result
    A=&RigUnit_GetTransform(Item=(Type=Bone,Name="spine_05"), Space=LocalSpace, bInitial=False).Transform
   ) {template="Multiply::Execute(in A,in B,out Result)"}
   n1 = &RigVMFunction_MathQuaternionSwingTwist(TwistAxis=(X=1,Y=0,Z=0), Input=@Multiply.Result.Rotation)
   > RigUnit_SetRotation = &RigUnit_SetRotation(Item=(Type=Bone,Name="deform_spine_04"), Weight=1, bInitial=False, Value=@n1.Twist)
   > &RigUnit_SetRotation(Item=(Type=Bone,Name="deform_spine_03"), Weight=1, bInitial=False, Value=@n1.Twist)
   > &RigUnit_SetRotation(Item=(Type=Bone,Name="deform_spine_02"), Weight=1, bInitial=False, Value=@n1.Twist)
   > &RigUnit_SetRotation(Item=(Type=Bone,Name="deform_spine_01"), Weight=1, bInitial=False, Value=@n1.Twist)
   > &RigUnit_SetRotation(Item=(Type=Bone,Name="deform_pelvis"), Weight=1, bInitial=False, Value=@n1.Twist)
   > RigVMFunctionReturnNode()
  ]}
 New Function_ContainedGraph {nodes [
   n1 = RigVMFunctionEntryNode()
   RigVMFunction_MathQuaternionToEuler = &RigVMFunction_MathQuaternionToEuler(
    RotationOrder=XYZ
    Value=&RigUnit_GetRelativeTransformForItem(
     Child=(Type=Bone,Name="hand_l")
     Parent=(Type=Null,Name="lowerarm_l")
     bChildInitial=False
     bParentInitial=False
     Child.Name=@n1.Child
     Parent.Name=@n1.Parent
    ).RelativeTransform.Rotation
   )
   Remap = &RigVMFunction_MathDoubleRemap(
    TargetMaximum=1
    TargetMinimum=0
    Value=@RigVMFunction_MathQuaternionToEuler.Result.X
    SourceMaximum=@n1.Source Maximum
    SourceMinimum=@n1.Source Minimum
   )
   > RigVMFunctionReturnNode(Delta Rotation=@Remap.Result)
  ]}
 RigVMFunctionLibrary {nodes [
   ApplyFBIKChainCorrection = RigVMCollapseNode(End=(Type=None,Name="None"), ExecuteContext=(), Start=(Type=None,Name="None"))
   Deform Elbows = RigVMCollapseNode(ExecuteContext=())
   Deform Spine = RigVMCollapseNode(ExecuteContext=())
   Delta Rotation X = RigVMCollapseNode(ExecuteContext=())
   Retarget Arms = RigVMCollapseNode(ExecuteContext=())
   Retarget_Arms = RigVMCollapseNode(ExecuteContext=())
   Retarget_fingers = RigVMCollapseNode(ExecuteContext=())
  ]}
 RigVMFunction_Sequence_1_ContainedGraph {nodes [
   n1 = RigVMFunctionEntryNode()
   RigVMFunction_Sequence = &RigVMFunction_Sequence(ExecuteContext=@n1.ExecuteContext)
   n2 = &RigVMFunction_Sequence(ExecuteContext=@RigVMFunction_Sequence.B)
   n3 = &RigVMFunction_Sequence(ExecuteContext=@n2.B)
   n4 = &RigVMFunction_Sequence(ExecuteContext=@n3.B)
   RigVMFunctionReturnNode(A=@RigVMFunction_Sequence.A, B=@n2.A, C=@n3.A, D=@n4.A, E=@n4.B)
  ]}
 RigVMFunction_Sequence_1_ContainedGraph {nodes [
   n1 = RigVMFunctionEntryNode()
   RigVMFunction_Sequence = &RigVMFunction_Sequence(ExecuteContext=@n1.ExecuteContext)
   n2 = &RigVMFunction_Sequence(ExecuteContext=@RigVMFunction_Sequence.B)
   n3 = &RigVMFunction_Sequence(ExecuteContext=@n2.B)
   RigVMFunctionReturnNode(A=@RigVMFunction_Sequence.A, B=@n2.A, C=@n3.A, D=@n3.B)
  ]}
 RigVMFunction_Sequence_1_ContainedGraph {nodes [
   n1 = RigVMFunctionEntryNode()
   RigVMFunction_Sequence = &RigVMFunction_Sequence(ExecuteContext=@n1.ExecuteContext)
   n2 = &RigVMFunction_Sequence(ExecuteContext=@RigVMFunction_Sequence.B)
   n3 = &RigVMFunction_Sequence(ExecuteContext=@n2.B)
   RigVMFunctionReturnNode(A=@RigVMFunction_Sequence.A, B=@n2.A, C=@n3.A, D=@n3.B)
  ]}
 RigVMFunction_Sequence_2_ContainedGraph {nodes [
   n1 = RigVMFunctionEntryNode()
   n2 = &RigVMFunction_Sequence(ExecuteContext=@n1.ExecuteContext)
   n3 = &RigVMFunction_Sequence(ExecuteContext=@n2.B)
   n4 = &RigVMFunction_Sequence(ExecuteContext=@n3.B)
   RigVMFunctionReturnNode(A=@n2.A, B=@n3.A, C=@n4.A, D=@n4.B)
  ]}
 RigVMModel {nodes [
   Delta Rotation X = RigVMFunctionReferenceNode(Child=hand_l, ExecuteContext=(), Parent=hand_l_null, Source Maximum=90)
   > &RigUnit_ModifyTransforms(
    ItemToModify="((Item=(Type=Bone,Name=\"deform_hand_up_l\"),Transform=(Rotation=(X=0,Y=0,Z=0,W=1),Translation=(X=0,Y=8,Z=0),Scale3D=(X=1,Y=1,Z=1))),(Item=(Type=Bone,Name=\"deform_hand_dw_l\"),Transform=(Rotation=(X=0,Y=0,Z=0,W=1),Translation=(X=0,Y=-6,Z=0),Scale3D=(X=1,Y=1,Z ...(truncated)"
    WeightMinimum=0
    Weight=@"Delta Rotation X".Delta Rotation
   )
   > Delta Rotation X_1 = RigVMFunctionReferenceNode(Child=hand_l, Parent=hand_l_null, Source Maximum=-90)
   > &RigUnit_ModifyTransforms(
    ItemToModify="((Item=(Type=Bone,Name=\"deform_hand_up_l\"),Transform=(Rotation=(X=0,Y=0,Z=0,W=1),Translation=(X=0,Y=8,Z=0),Scale3D=(X=1,Y=1,Z=1))),(Item=(Type=Bone,Name=\"deform_hand_dw_l\"),Transform=(Rotation=(X=0,Y=0,Z=0,W=1),Translation=(X=0,Y=-6,Z=0),Scale3D=(X=1,Y=1,Z ...(truncated)"
    WeightMinimum=0
    Weight=@"Delta Rotation X_1".Delta Rotation
   )
   > Delta Rotation X_2 = RigVMFunctionReferenceNode(Child=hand_r, Parent=hand_r_null, Source Maximum=90)
   > &RigUnit_ModifyTransforms(
    ItemToModify="((Item=(Type=Bone,Name=\"deform_hand_up_r\"),Transform=(Rotation=(X=0,Y=0,Z=0,W=1),Translation=(X=0,Y=-8,Z=0),Scale3D=(X=1,Y=1,Z=1))),(Item=(Type=Bone,Name=\"deform_hand_dw_r\"),Transform=(Rotation=(X=0,Y=0,Z=0,W=1),Translation=(X=0,Y=6,Z=0),Scale3D=(X=1,Y=1,Z ...(truncated)"
    WeightMinimum=0
    Weight=@"Delta Rotation X_2".Delta Rotation
   )
   > Delta Rotation X_3 = RigVMFunctionReferenceNode(Child=hand_r, Parent=hand_r_null, Source Maximum=-90)
   > &RigUnit_ModifyTransforms(
    ItemToModify="((Item=(Type=Bone,Name=\"deform_hand_up_r\"),Transform=(Rotation=(X=0,Y=0,Z=0,W=1),Translation=(X=0,Y=-8,Z=0),Scale3D=(X=1,Y=1,Z=1))),(Item=(Type=Bone,Name=\"deform_hand_dw_r\"),Transform=(Rotation=(X=0,Y=0,Z=0,W=1),Translation=(X=0,Y=6,Z=0),Scale3D=(X=1,Y=1,Z ...(truncated)"
    WeightMinimum=0
    Weight=@"Delta Rotation X_3".Delta Rotation
   )
   n1 = &RigUnit_PrepareForExecution()
   addFKCtrl = RigVMFunctionReferenceNode(
    Custom_FKCtrl=()
    Options_FKCtrl="(pelvis_1_#1=true,spine_3_#2=true,clavicle_5_#3=true,arm_7_#4=true,leg_9_#5=False,head_11_46DC15EF465CD297 ...(truncated)"
    Shape_Settings="(pelvis_2_#6=(bVisible=True,Name=\"Hexagon_Thin\",Color=(R=0.0117,G=0.8281,B=1,A=1),Transform=(Rotation=(X=0,Y=-0.7071,Z=0,W=0.7071),Translation=(X=2,Y=0,Z=0),Scale3D=(X=5,Y=5,Z=7))),spine_4_#7=(bVi ...(truncated)"
    ExecuteContext=@n1.ExecutePin
   )
   RigUnit_BeginExecution = &RigUnit_BeginExecution()
   RigVMFunction_Sequence = RigVMAggregateNode(ExecuteContext=@RigUnit_BeginExecution.ExecutePin)
   FN_DeformRig_FNC = RigVMFunctionReferenceNode(
    DeformRig_ Curve="(deformRig_l_upperarm_up_alpha_72_#8=false,deformRig_r_upperarm_up_alpha_73_#9=False,deformRig_l_upperarm_fr_alpha_74_#10=False,deformRig_r_upperarm_fr_alpha_75_CB51B8 ...(truncated)"
    DeformRig_Bool="(upperarm_1_#11=True,elbow_3_#12=True,wrist_7_#13=True,groin_8_#14=True,glute_17_#15=True,knee_9_40C408324CA261 ...(truncated)"
    DeformRig_Debug=(upperarm_debug_18_#11=False,wrist_debug_20_#13=False,lat_debug_26_#16=False,groinGlute_debug_27_#14=False)
    DeformRig_Value="(upperarm_up_18_#17=(Rotation=(X=0,Y=0,Z=0,W=1),Translation=(X=0,Y=0,Z=8),Scale3D=(X=0.7,Y=1,Z=1)),upperarm_fwd_fr_19_#18=(Rotation=(X=0,Y=0,Z=0,W=1),Translation=(X=0,Y=10,Z=0),Scale3D=(X=0.5,Y=1,Z= ...(truncated)"
    ExecuteContext=@RigVMFunction_Sequence.A
   )
   Retarget Arms = RigVMFunctionReferenceNode(ExecuteContext=@RigVMFunction_Sequence.B)
   > &RigUnit_RotationConstraintLocalSpaceOffset(
    AdvancedSettings=(InterpolationType=Average,RotationOrderForFilter=XZY)
    Child=(Type=Bone,Name="deform_thigh_l")
    Filter=(bX=True,bY=True,bZ=True)
    Parents=((Item=(Type=Bone,Name="thigh_l"),Weight=1))
    Weight=1
   )
   > &RigUnit_RotationConstraintLocalSpaceOffset(
    AdvancedSettings=(InterpolationType=Average,RotationOrderForFilter=XZY)
    Child=(Type=Bone,Name="deform_thigh_r")
    Filter=(bX=True,bY=True,bZ=True)
    Parents=((Item=(Type=Bone,Name="thigh_r"),Weight=1))
    Weight=1
   )
   RigVMFunction_Sequence_1 = RigVMAggregateNode(ExecuteContext=@RigVMFunction_Sequence.D)
   n2 = &RigUnit_SphericalPoseReader(
    ActiveRegionScaleFactors=(PositiveWidth=1.2,NegativeWidth=1,PositiveHeight=1.2,NegativeHeight=1.2)
    ActiveRegionSize=0.12
    Debug=(bDrawDebug=false,bDraw2D=False,bDrawLocalAxes=False,DebugScale=25,DebugSegments=20,DebugThickness=0.25)
    DriverAxis=(X=1,Y=0,Z=0)
    DriverItem=(Type=Bone,Name="neck_01")
    FalloffRegionScaleFactors=(PositiveWidth=0.8,NegativeWidth=1,PositiveHeight=1,NegativeHeight=1)
    FalloffSize=0.3
    FlipHeightScaling=False
    FlipWidthScaling=False
    OptionalParentItem=(Type=Bone,Name="None")
    RotationOffset=(X=-90,Y=0,Z=35)
    ExecutePin=@RigVMFunction_Sequence_1.A
   )
   > RigVMVariableNode(Value=@n2.OutputParam) {variable=neckback Alpha}
   > n3 = &RigUnit_SphericalPoseReader(
    ActiveRegionScaleFactors=(PositiveWidth=1,NegativeWidth=1,PositiveHeight=1,NegativeHeight=1)
    ActiveRegionSize=0.15
    Debug=(bDrawDebug=false,bDraw2D=False,bDrawLocalAxes=False,DebugScale=25,DebugSegments=20,DebugThickness=0.25)
    DriverAxis=(X=1,Y=0,Z=0)
    DriverItem=(Type=Bone,Name="neck_01")
    FalloffRegionScaleFactors=(PositiveWidth=2,NegativeWidth=1,PositiveHeight=1,NegativeHeight=1.5)
    FalloffSize=0.2
    FlipHeightScaling=False
    FlipWidthScaling=False
    OptionalParentItem=(Type=Bone,Name="None")
    RotationOffset=(X=-50,Y=0,Z=50)
   )
   > RigVMVariableNode(Value=@n3.OutputParam) {variable=R neckside Alpha}
   > n4 = &RigUnit_SphericalPoseReader(
    ActiveRegionScaleFactors=(PositiveWidth=1,NegativeWidth=1,PositiveHeight=1,NegativeHeight=1)
    ActiveRegionSize=0.15
    Debug=(bDrawDebug=false,bDraw2D=False,bDrawLocalAxes=False,DebugScale=25,DebugSegments=20,DebugThickness=0.25)
    DriverAxis=(X=1,Y=0,Z=0)
    DriverItem=(Type=Bone,Name="neck_01")
    FalloffRegionScaleFactors=(PositiveWidth=2,NegativeWidth=1,PositiveHeight=1.5,NegativeHeight=1)
    FalloffSize=0.2
    FlipHeightScaling=False
    FlipWidthScaling=False
    OptionalParentItem=(Type=Bone,Name="None")
    RotationOffset=(X=-130,Y=0,Z=50)
   )
   > RigVMVariableNode(Value=@n4.OutputParam) {variable=L neckside Alpha}
   > n5 = &RigUnit_SphericalPoseReader(
    ActiveRegionScaleFactors=(PositiveWidth=2,NegativeWidth=1,PositiveHeight=2,NegativeHeight=2)
    ActiveRegionSize=0.1
    Debug=(bDrawDebug=false,bDraw2D=False,bDrawLocalAxes=False,DebugScale=25,DebugSegments=20,DebugThickness=0.25)
    DriverAxis=(X=-1,Y=0,Z=0)
    DriverItem=(Type=Bone,Name="thigh_l")
    FalloffRegionScaleFactors=(PositiveWidth=1,NegativeWidth=1.5,PositiveHeight=1,NegativeHeight=1)
    FalloffSize=0.25
    FlipHeightScaling=False
    FlipWidthScaling=False
    OptionalParentItem=(Type=Bone,Name="None")
    RotationOffset=(X=110,Y=0,Z=20)
   )
   > RigVMVariableNode(Value=@n5.OutputParam) {variable=L thighfront Alpha}
   > n6 = &RigUnit_SphericalPoseReader(
    ActiveRegionScaleFactors=(PositiveWidth=1,NegativeWidth=2,PositiveHeight=2,NegativeHeight=2)
    ActiveRegionSize=0.1
    Debug=(bDrawDebug=false,bDraw2D=False,bDrawLocalAxes=False,DebugScale=25,DebugSegments=20,DebugThickness=0.25)
    DriverAxis=(X=1,Y=0,Z=0)
    DriverItem=(Type=Bone,Name="thigh_r")
    FalloffRegionScaleFactors=(PositiveWidth=1.5,NegativeWidth=1,PositiveHeight=1,NegativeHeight=1)
    FalloffSize=0.25
    FlipHeightScaling=False
    FlipWidthScaling=False
    OptionalParentItem=(Type=Bone,Name="None")
    RotationOffset=(X=-80,Y=0,Z=20)
   )
   > RigVMVariableNode(Value=@n6.OutputParam) {variable=R thighfront Alpha}
   n7 = RigVMVariableNode() {variable=neckback Alpha}
   &RigUnit_ModifyTransforms(
    ItemToModify=((Item=(Type=Bone,Name="deform_collar_01_back"),Transform=(Rotation=(X=0,Y=0,Z=0.4226,W=0.9063),Translation=(X=0,Y=0,Z=0),Scale3D=(X=1,Y=1,Z=1))))
    WeightMinimum=0
    Weight=@n7.Value
    ExecutePin=@RigVMFunction_Sequence_1.B
   )
   > &RigUnit_ModifyTransforms(
    ItemToModify="((Item=(Type=Bone,Name=\"deform_collar_01_back_r\"),Transform=(Rotation=(X=0,Y=0,Z=0.4226,W=0.9063),Translation=(X=0,Y=0,Z=0),Scale3D=(X=1,Y=1,Z=1))),(Item=(Type=Bone,Name=\"deform_collar_02_back_r\"),Transform=(Rotation=(X=0,Y=0,Z=0.0872,W=0.9962),Translation ...(truncated)"
    WeightMinimum=0
    Weight=RigVMVariableNode() {variable=R neckside Alpha}.Value
   )
   > &RigUnit_ModifyTransforms(
    ItemToModify="((Item=(Type=Bone,Name=\"deform_collar_01_back_l\"),Transform=(Rotation=(X=0,Y=0,Z=0.4226,W=0.9063),Translation=(X=0,Y=0,Z=0),Scale3D=(X=1,Y=1,Z=1))),(Item=(Type=Bone,Name=\"deform_collar_02_back_l\"),Transform=(Rotation=(X=0,Y=0,Z=0.0872,W=0.9962),Translation ...(truncated)"
    WeightMinimum=0
    Weight=RigVMVariableNode() {variable=L neckside Alpha}.Value
   )
   &RigUnit_ModifyTransforms(
    ItemToModify=((Item=(Type=Bone,Name="deform_thighpush_base_l"),Transform=(Rotation=(X=0,Y=0,Z=0,W=1),Translation=(X=5,Y=-5,Z=0),Scale3D=(X=1,Y=1,Z=1))))
    WeightMinimum=0
    Weight=RigVMVariableNode() {variable=L thighfront Alpha}.Value
    ExecutePin=@RigVMFunction_Sequence_1.C
   )
   > &RigUnit_ModifyTransforms(
    ItemToModify=((Item=(Type=Bone,Name="deform_thighpush_base_r"),Transform=(Rotation=(X=0,Y=0,Z=0,W=1),Translation=(X=-5,Y=5,Z=0),Scale3D=(X=1,Y=1,Z=1))))
    WeightMinimum=0
    Weight=RigVMVariableNode() {variable=R thighfront Alpha}.Value
   )
   RigVMFunction_ControlFlowBranch = &RigVMFunction_ControlFlowBranch(Condition=RigVMVariableNode() {variable=Is SkyDiving}.Value, ExecuteContext=@RigVMFunction_Sequence_1.D)
   n8 = &RigUnit_SphericalPoseReader(
    ActiveRegionScaleFactors=(PositiveWidth=3,NegativeWidth=1,PositiveHeight=1.2,NegativeHeight=1)
    ActiveRegionSize=0.1
    Debug=(bDrawDebug=false,bDraw2D=False,bDrawLocalAxes=False,DebugScale=25,DebugSegments=20,DebugThickness=0.25)
    DriverAxis=(X=-1,Y=0,Z=0)
    DriverItem=(Type=Bone,Name="deform_upperarm_r")
    FalloffRegionScaleFactors=(PositiveWidth=1,NegativeWidth=1,PositiveHeight=1,NegativeHeight=1)
    FalloffSize=0.3
    FlipHeightScaling=False
    FlipWidthScaling=False
    OptionalParentItem=(Type=Bone,Name="None")
    RotationOffset=(X=-150,Y=40,Z=0)
    ExecutePin=@RigVMFunction_ControlFlowBranch.False
   )
   > &RigUnit_ModifyTransforms(
    ItemToModify="((Item=(Type=Bone,Name=\"deform_shouldersqueeze_r\"),Transform=(Rotation=(X=0,Y=0,Z=0,W=1),Translation=(X=3,Y=0,Z=0),Scale3D=(X=1,Y=1,Z=1))),(Item=(Type=Bone,Name=\"deform_uppershoulder_r\"),Transform=(Rotation=(X=0,Y=0,Z=0,W=1),Translation=(X=3,Y=0,Z=0),Scale ...(truncated)"
    WeightMinimum=0
    Weight=@n8.OutputParam
   )
   > n9 = &RigUnit_SphericalPoseReader(
    ActiveRegionScaleFactors=(PositiveWidth=3,NegativeWidth=1,PositiveHeight=1.2,NegativeHeight=1.5)
    ActiveRegionSize=0.05
    Debug=(bDrawDebug=false,bDraw2D=False,bDrawLocalAxes=False,DebugScale=25,DebugSegments=20,DebugThickness=0.25)
    DriverAxis=(X=-1,Y=0,Z=0)
    DriverItem=(Type=Bone,Name="deform_upperarm_r")
    FalloffRegionScaleFactors=(PositiveWidth=1,NegativeWidth=1,PositiveHeight=1,NegativeHeight=1)
    FalloffSize=0.2
    FlipHeightScaling=False
    FlipWidthScaling=False
    OptionalParentItem=(Type=Bone,Name="None")
    RotationOffset=(X=-100,Y=40,Z=0)
   )
   > &RigUnit_ModifyTransforms(
    ItemToModify="((Item=(Type=Bone,Name=\"deform_chestsqueeze_r\"),Transform=(Rotation=(X=0,Y=0,Z=0,W=1),Translation=(X=10,Y=-5,Z=0),Scale3D=(X=1,Y=1,Z=1))),(Item=(Type=Bone,Name=\"deform_shouldersqueeze_front_r\"),Transform=(Rotation=(X=0,Y=0,Z=0,W=1),Translation=(X=5,Y=0,Z=0 ...(truncated)"
    WeightMinimum=0
    Weight=@n9.OutputParam
   )
   > n10 = &RigUnit_SphericalPoseReader(
    ActiveRegionScaleFactors=(PositiveWidth=3,NegativeWidth=1,PositiveHeight=1.2,NegativeHeight=1)
    ActiveRegionSize=0.1
    Debug=(bDrawDebug=false,bDraw2D=False,bDrawLocalAxes=False,DebugScale=25,DebugSegments=20,DebugThickness=0.25)
    DriverAxis=(X=1,Y=0,Z=0)
    DriverItem=(Type=Bone,Name="deform_upperarm_l")
    FalloffRegionScaleFactors=(PositiveWidth=1,NegativeWidth=1,PositiveHeight=1,NegativeHeight=1)
    FalloffSize=0.3
    FlipHeightScaling=true
    FlipWidthScaling=true
    OptionalParentItem=(Type=Bone,Name="None")
    RotationOffset=(X=-330,Y=40,Z=0)
   )
   > &RigUnit_ModifyTransforms(
    ItemToModify="((Item=(Type=Bone,Name=\"deform_shouldersqueeze_l\"),Transform=(Rotation=(X=0,Y=0,Z=0,W=1),Translation=(X=-3,Y=0,Z=0),Scale3D=(X=1,Y=1,Z=1))),(Item=(Type=Bone,Name=\"deform_uppershoulder_l\"),Transform=(Rotation=(X=0,Y=0,Z=0,W=1),Translation=(X=-3,Y=0,Z=0),Sca ...(truncated)"
    WeightMinimum=0
    Weight=@n10.OutputParam
   )
   > n11 = &RigUnit_SphericalPoseReader(
    ActiveRegionScaleFactors=(PositiveWidth=3,NegativeWidth=1,PositiveHeight=1.2,NegativeHeight=1.5)
    ActiveRegionSize=0.05
    Debug=(bDrawDebug=false,bDraw2D=False,bDrawLocalAxes=False,DebugScale=25,DebugSegments=20,DebugThickness=0.25)
    DriverAxis=(X=1,Y=0,Z=0)
    DriverItem=(Type=Bone,Name="deform_upperarm_l")
    FalloffRegionScaleFactors=(PositiveWidth=1,NegativeWidth=1,PositiveHeight=1,NegativeHeight=1)
    FalloffSize=0.2
    FlipHeightScaling=true
    FlipWidthScaling=true
    OptionalParentItem=(Type=Bone,Name="None")
    RotationOffset=(X=80,Y=40,Z=0)
   )
   > &RigUnit_ModifyTransforms(
    ItemToModify="((Item=(Type=Bone,Name=\"deform_chestsqueeze_l\"),Transform=(Rotation=(X=0,Y=0,Z=0,W=1),Translation=(X=-10,Y=-5,Z=0),Scale3D=(X=1,Y=1,Z=1))),(Item=(Type=Bone,Name=\"deform_shouldersqueeze_front_l\"),Transform=(Rotation=(X=0,Y=0,Z=0,W=1),Translation=(X=-5,Y=0,Z ...(truncated)"
    WeightMinimum=0
    Weight=@n11.OutputParam
   )
   n12 = &RigUnit_SphericalPoseReader(
    ActiveRegionScaleFactors=(PositiveWidth=2,NegativeWidth=2,PositiveHeight=1.5,NegativeHeight=3)
    ActiveRegionSize=0.1
    Debug=(bDrawDebug=false,bDraw2D=False,bDrawLocalAxes=False,DebugScale=25,DebugSegments=20,DebugThickness=0.25)
    DriverAxis=(X=-1,Y=0,Z=0)
    DriverItem=(Type=Bone,Name="deform_upperarm_r")
    FalloffRegionScaleFactors=(PositiveWidth=0.5,NegativeWidth=1.5,PositiveHeight=1,NegativeHeight=1)
    FalloffSize=0.35
    FlipHeightScaling=False
    FlipWidthScaling=False
    OptionalParentItem=(Type=Bone,Name="None")
    RotationOffset=(X=110,Y=0,Z=0)
    ExecutePin=@RigVMFunction_ControlFlowBranch.True
   )
   > &RigUnit_ModifyTransforms(
    ItemToModify="((Item=(Type=Bone,Name=\"deform_uppershoulder_r\"),Transform=(Rotation=(X=0,Y=0,Z=-0.2588,W=0.9659),Translation=(X=0,Y=0,Z=0),Scale3D=(X=1,Y=1,Z=1))),(Item=(Type=Bone,Name=\"deform_shouldersqueeze_r\"),Transform=(Rotation=(X=0,Y=0,Z=-0.0872,W=0.9962),Translati ...(truncated)"
    WeightMinimum=0
    Weight=@n12.OutputParam
   )
   > n13 = &RigUnit_SphericalPoseReader(
    ActiveRegionScaleFactors=(PositiveWidth=0.5,NegativeWidth=0.5,PositiveHeight=1,NegativeHeight=1)
    ActiveRegionSize=0.2
    Debug=(bDrawDebug=true,bDraw2D=False,bDrawLocalAxes=False,DebugScale=25,DebugSegments=20,DebugThickness=0.25)
    DriverAxis=(X=-1,Y=0,Z=0)
    DriverItem=(Type=Bone,Name="deform_upperarm_r")
    FalloffRegionScaleFactors=(PositiveWidth=1,NegativeWidth=1,PositiveHeight=1,NegativeHeight=0.5)
    FalloffSize=0.4
    FlipHeightScaling=False
    FlipWidthScaling=False
    OptionalParentItem=(Type=Bone,Name="None")
    RotationOffset=(X=-20,Y=220,Z=0)
   )
   > &RigUnit_ParentConstraint(
    AdvancedSettings=(InterpolationType=Average,RotationOrderForFilter=XZY)
    Child=(Type=Bone,Name="deform_uppershoulder_r")
    Filter=(TranslationFilter=(bX=false,bY=false,bZ=false),RotationFilter=(bX=false,bY=true,bZ=false),ScaleFilter=(bX=True,bY=True,bZ=True))
    Parents=((Item=(Type=Bone,Name="deform_upperarm_r"),Weight=1))
    Weight=@n13.OutputParam
   )
   > n14 = &RigUnit_SphericalPoseReader(
    ActiveRegionScaleFactors=(PositiveWidth=2,NegativeWidth=2,PositiveHeight=1.5,NegativeHeight=3)
    ActiveRegionSize=0.1
    Debug=(bDrawDebug=false,bDraw2D=False,bDrawLocalAxes=False,DebugScale=25,DebugSegments=20,DebugThickness=0.25)
    DriverAxis=(X=1,Y=0,Z=0)
    DriverItem=(Type=Bone,Name="deform_upperarm_l")
    FalloffRegionScaleFactors=(PositiveWidth=0.5,NegativeWidth=1.5,PositiveHeight=1,NegativeHeight=1)
    FalloffSize=0.35
    FlipHeightScaling=false
    FlipWidthScaling=true
    OptionalParentItem=(Type=Bone,Name="None")
    RotationOffset=(X=-80,Y=0,Z=0)
   )
   > &RigUnit_ModifyTransforms(
    ItemToModify="((Item=(Type=Bone,Name=\"deform_uppershoulder_l\"),Transform=(Rotation=(X=0,Y=0,Z=-0.2588,W=0.9659),Translation=(X=0,Y=0,Z=0),Scale3D=(X=1,Y=1,Z=1))),(Item=(Type=Bone,Name=\"deform_shouldersqueeze_l\"),Transform=(Rotation=(X=0,Y=0,Z=0.0872,W=0.9962),Translatio ...(truncated)"
    WeightMinimum=0
    Weight=@n14.OutputParam
   )
   > n15 = &RigUnit_SphericalPoseReader(
    ActiveRegionScaleFactors=(PositiveWidth=0.5,NegativeWidth=0.5,PositiveHeight=1,NegativeHeight=1)
    ActiveRegionSize=0.2
    Debug=(bDrawDebug=true,bDraw2D=False,bDrawLocalAxes=False,DebugScale=25,DebugSegments=20,DebugThickness=0.25)
    DriverAxis=(X=1,Y=0,Z=0)
    DriverItem=(Type=Bone,Name="deform_upperarm_l")
    FalloffRegionScaleFactors=(PositiveWidth=1,NegativeWidth=1,PositiveHeight=1,NegativeHeight=0.5)
    FalloffSize=0.4
    FlipHeightScaling=true
    FlipWidthScaling=true
    OptionalParentItem=(Type=Bone,Name="None")
    RotationOffset=(X=20,Y=40,Z=0)
   )
   > &RigUnit_ParentConstraint(
    AdvancedSettings=(InterpolationType=Average,RotationOrderForFilter=XZY)
    Child=(Type=Bone,Name="deform_uppershoulder_l")
    Filter=(TranslationFilter=(bX=false,bY=false,bZ=false),RotationFilter=(bX=false,bY=true,bZ=false),ScaleFilter=(bX=True,bY=True,bZ=True))
    Parents=((Item=(Type=Bone,Name="deform_upperarm_l"),Weight=1))
    Weight=@n15.OutputParam
   )
   &RigUnit_AimConstraintLocalSpaceOffset(
    Child=(Type=Bone,Name="deform_chestsqueeze_r")
    Filter=(bX=True,bY=True,bZ=True)
    Parents=((Item=(Type=Bone,Name="deform_chestsqueeze_target_r"),Weight=1))
    Weight=1
    WorldUp=(Target=(X=0,Y=0,Z=1),Kind=Location,Space=(Type=Bone,Name="None"))
    ExecutePin=@RigVMFunction_ControlFlowBranch.Completed
   )
   > &RigUnit_AimConstraintLocalSpaceOffset(
    AimAxis=(X=-1,Y=0,Z=0)
    Child=(Type=Bone,Name="deform_chestsqueeze_l")
    Filter=(bX=True,bY=True,bZ=True)
    Parents=((Item=(Type=Bone,Name="deform_chestsqueeze_target"),Weight=1))
    Weight=1
    WorldUp=(Target=(X=0,Y=0,Z=1),Kind=Location,Space=(Type=Bone,Name="None"))
   )
   Add = &RigVMFunction_MathDoubleAdd(A=RigVMVariableNode() {variable=L neckside Alpha}.Value, B=RigVMVariableNode() {variable=R neckside Alpha}.Value)
   Remap = &RigVMFunction_MathDoubleRemap(SourceMaximum=1, SourceMinimum=0, TargetMaximum=0.5, TargetMinimum=0, Value=@Add.Result)
   &RigVMFunction_MathDoubleAdd(B=@Remap.Result, A=@n7.Value)
   RigVMRerouteNode(Value=@RigVMFunction_Sequence.C)
   setFKCtrl = RigVMFunctionReferenceNode(ExecuteContext=(), setFK_Ctrl=false)
  ]}
]
hierarchy [
 root {loc=<0 0 0>, rot=<0 0 0 1>}
 pelvis {parent=root, loc=<0 -0.7321 92.4218>, rot=<0.0127 0.707 -0.0127 -0.707>}
 spine_01 {parent=pelvis, loc=<5.6002 -0.5128 0>, rot=<0 0 0.034 -0.9994>}
 spine_02 {parent=spine_01, loc=<8.0756 0.0139 0>, rot=<0 0 0.0238 -0.9997>}
 spine_03 {parent=spine_02, loc=<7.0395 0 0>, rot=<0 0 0.0149 -0.9999>}
 spine_04 {parent=spine_03, loc=<8.1023 0 0>, rot=<0 0 -0.0132 -0.9999>}
 spine_05 {parent=spine_04, loc=<10.0141 0 0>, rot=<0 0 -0.0015 -1>}
 neck_01 {parent=spine_05, loc=<10.815 1.9196 0>, rot=<0 0 -0.1864 -0.9825>}
 neck_02 {parent=neck_01, loc=<4.7347 0.3045 0>, rot=<0 0 0.0363 -0.9993>}
 head {parent=neck_02, loc=<5.7632 0.5552 0>, rot=<0 0 0.115 -0.9934>}
 attach_cape {parent=spine_05, loc=<9.017 -11.714 0>, rot=<0.0616 -0.7044 -0.0616 -0.7044>}
 cape {parent=attach_cape, type=Null, loc=<0 0 0>, rot=<0 0 0 1>}
 attach_backpack {parent=spine_05, loc=<-0.123 -13.473 0>, rot=<-0.0093 -0.707 0.0093 -0.707>}
 Backpack_BR {parent=attach_backpack, type=Null, loc=<0 -4.9009 0.337>, rot=<-0.0872 0 0 0.9962>}
 clavicle_l {parent=spine_05, loc=<2.5856 -0.2866 -4.3777>, rot=<0.0465 -0.7289 -0.009 -0.6829>}
 upperarm_l {parent=clavicle_l, loc=<18.2264 0.1342 -0.001>, rot=<0.0253 -0.2693 0.0013 -0.9627>}
 lowerarm_l {parent=upperarm_l, loc=<29.0071 0.1686 -1.239>, rot=<0.0391 -0.0099 -0.2453 -0.9686>}
 hand_l {parent=lowerarm_l, loc=<25.8607 0.222 1.0274>, rot=<-0.7193 -0.0988 0.1193 -0.6773>}
 thumb_01_l {parent=hand_l, loc=<1.3118 -0.3191 -2.459>, rot=<0.6546 -0.4042 0.0664 -0.6354>}
 thumb_02_l {parent=thumb_01_l, loc=<5.9336 -0.0448 0.0401>, rot=<-0.0037 0.0213 0.298 -0.9543>}
 thumb_03_l {parent=thumb_02_l, loc=<2.2223 0.3682 -0.0907>, rot=<0.0118 -0.013 0.0066 -0.9998>}
 index_metacarpal_l {parent=hand_l, loc=<1.6535 0.0183 -2.96>, rot=<0.0158 -0.0124 -0.0173 -0.9996>}
 index_01_l {parent=index_metacarpal_l, loc=<8.3008 0.0357 -0.0524>, rot=<-0.0044 0.0485 0.0924 -0.9945>}
 index_02_l {parent=index_01_l, loc=<3.3587 -0.3004 0.0452>, rot=<-0.002 0.0027 0.3225 -0.9466>}
 index_03_l {parent=index_02_l, loc=<2.5578 -0.123 -0.0252>, rot=<-0.002 0.0067 0.2558 -0.9667>}
 middle_metacarpal_l {parent=hand_l, loc=<2.1856 -0.1725 -0.5984>, rot=<0.006 0.0168 -0.0329 -0.9993>}
 middle_01_l {parent=middle_metacarpal_l, loc=<7.3275 -0.2447 0.0129>, rot=<-0.0208 0.0042 0.0858 -0.9961>}
 middle_02_l {parent=middle_01_l, loc=<4.0123 0.1401 0.1175>, rot=<-0.0218 0.0258 0.3512 -0.9357>}
 middle_03_l {parent=middle_02_l, loc=<2.6077 0.0534 0.1373>, rot=<0.0082 0.0188 0.1708 -0.9851>}
 ring_metacarpal_l {parent=hand_l, loc=<2.2074 -0.3625 1.3638>, rot=<-0.0284 0.0722 -0.0458 -0.9959>}
 ring_01_l {parent=ring_metacarpal_l, loc=<7.0295 -0.6197 -0.1446>, rot=<-0.0489 -0.0252 0.1892 -0.9804>}
 ring_02_l {parent=ring_01_l, loc=<3.6174 0.4647 -0.1027>, rot=<0 0 0.2806 -0.9598>}
 ring_03_l {parent=ring_02_l, loc=<2.4152 0.114 0.0199>, rot=<0 0 0.1189 -0.9929>}
 pinky_metacarpal_l {parent=hand_l, loc=<1.7616 -0.9753 3.3814>, rot=<-0.0397 0.0814 -0.069 -0.9935>}
 pinky_01_l {parent=pinky_metacarpal_l, loc=<7.1709 -0.4831 -0.1094>, rot=<-0.0654 -0.0035 0.2242 -0.9723>}
 pinky_02_l {parent=pinky_01_l, loc=<2.8274 0.0893 -0.1043>, rot=<0.0037 0.0011 0.2854 -0.9584>}
 pinky_03_l {parent=pinky_02_l, loc=<2.2046 0.0307 -0.0078>, rot=<-0.0002 0.0098 0.1445 -0.9895>}
 hand_attach_l {parent=hand_l, loc=<8.5 -3 0>, rot=<0.7071 0.7071 0 0>}
 deform_wrist_dn_base_l {parent=hand_l, loc=<0 -3.4173 0>, rot=<0 0 0 -1>}
 deform_wrist_dn_l {parent=deform_wrist_dn_base_l, loc=<0 0 0>, rot=<0 0 0 1>}
 deform_wrist_up_base_l {parent=hand_l, loc=<0 2.7805 0>, rot=<0 0 0 -1>}
 deform_wrist_up_l {parent=deform_wrist_up_base_l, loc=<0 0 0>, rot=<0 0 0 1>}
 lowerarm_twist_02_l {parent=lowerarm_l, loc=<9.6199 0 0>, rot=<0 0 0 -1>}
 lowerarm_twist_01_l {parent=lowerarm_l, loc=<17.8549 0 0>, rot=<0 0 0 -1>}
 cr_lowerarm_push_l {parent=lowerarm_l, loc=<0 0 0>, rot=<0 0 0 1>}
 hand_l_null {parent=upperarm_l, type=Null, loc=<51.6459 12.7132 -1.2215>, rot=<0.6448 0.2741 0.0395 0.7124>}
 clavicle_r {parent=spine_05, loc=<2.5855 -0.2866 4.3777>, rot=<0.7289 0.0465 -0.6829 0.009>}
 upperarm_r {parent=clavicle_r, loc=<-18.2264 -0.1342 0.001>, rot=<0.0253 -0.2693 0.0013 -0.9627>}
 lowerarm_r {parent=upperarm_r, loc=<-29.0071 -0.1686 1.239>, rot=<0.0391 -0.0099 -0.2453 -0.9686>}
 hand_r {parent=lowerarm_r, loc=<-25.8607 -0.222 -1.0274>, rot=<-0.7193 -0.0988 0.1193 -0.6773>}
 thumb_01_r {parent=hand_r, loc=<-1.3118 0.3191 2.459>, rot=<0.6546 -0.4042 0.0664 -0.6354>}
 thumb_02_r {parent=thumb_01_r, loc=<-5.9336 0.0448 -0.0401>, rot=<-0.0037 0.0213 0.298 -0.9543>}
 thumb_03_r {parent=thumb_02_r, loc=<-2.2223 -0.3682 0.0907>, rot=<0.0118 -0.013 0.0066 -0.9998>}
 index_metacarpal_r {parent=hand_r, loc=<-1.6535 -0.0183 2.96>, rot=<0.0158 -0.0124 -0.0173 -0.9996>}
 index_01_r {parent=index_metacarpal_r, loc=<-8.3008 -0.0357 0.0524>, rot=<-0.0044 0.0485 0.0924 -0.9945>}
 index_02_r {parent=index_01_r, loc=<-3.3587 0.3004 -0.0452>, rot=<-0.002 0.0027 0.3225 -0.9466>}
 index_03_r {parent=index_02_r, loc=<-2.5578 0.123 0.0252>, rot=<-0.002 0.0067 0.2558 -0.9667>}
 middle_metacarpal_r {parent=hand_r, loc=<-2.1856 0.1725 0.5984>, rot=<0.006 0.0168 -0.0329 -0.9993>}
 middle_01_r {parent=middle_metacarpal_r, loc=<-7.3275 0.2447 -0.0129>, rot=<-0.0208 0.0042 0.0858 -0.9961>}
 middle_02_r {parent=middle_01_r, loc=<-4.0123 -0.1401 -0.1175>, rot=<-0.0218 0.0258 0.3512 -0.9357>}
 middle_03_r {parent=middle_02_r, loc=<-2.6077 -0.0534 -0.1373>, rot=<0.0082 0.0188 0.1708 -0.9851>}
 ring_metacarpal_r {parent=hand_r, loc=<-2.2074 0.3625 -1.3638>, rot=<-0.0284 0.0722 -0.0458 -0.9959>}
 ring_01_r {parent=ring_metacarpal_r, loc=<-7.0295 0.6197 0.1446>, rot=<-0.0489 -0.0252 0.1892 -0.9804>}
 ring_02_r {parent=ring_01_r, loc=<-3.6174 -0.4647 0.1027>, rot=<0 0 0.2806 -0.9598>}
 ring_03_r {parent=ring_02_r, loc=<-2.4152 -0.114 -0.0199>, rot=<0 0 0.1189 -0.9929>}
 pinky_metacarpal_r {parent=hand_r, loc=<-1.7616 0.9753 -3.3814>, rot=<-0.0397 0.0814 -0.069 -0.9935>}
 pinky_01_r {parent=pinky_metacarpal_r, loc=<-7.1709 0.4831 0.1094>, rot=<-0.0654 -0.0035 0.2242 -0.9723>}
 pinky_02_r {parent=pinky_01_r, loc=<-2.8274 -0.0893 0.1043>, rot=<0.0037 0.0011 0.2854 -0.9584>}
 pinky_03_r {parent=pinky_02_r, loc=<-2.2046 -0.0307 0.0078>, rot=<-0.0002 0.0098 0.1445 -0.9895>}
 hand_attach_r {parent=hand_r, loc=<-8.5 3 0>, rot=<0 0 -0.7071 -0.7071>}
 deform_wrist_dn_base_r {parent=hand_r, loc=<0 3.4172 0>, rot=<0 0 0 -1>}
 deform_wrist_dn_r {parent=deform_wrist_dn_base_r, loc=<0 0 0>, rot=<0 0 0 -1>}
 deform_wrist_up_base_r {parent=hand_r, loc=<-0.0001 -2.7804 0>, rot=<0 0 0 -1>}
 deform_wrist_up_r {parent=deform_wrist_up_base_r, loc=<0 0 0>, rot=<0 0 0 -1>}
 lowerarm_twist_02_r {parent=lowerarm_r, loc=<-9.6199 0 0>, rot=<0 0 0 -1>}
 lowerarm_twist_01_r {parent=lowerarm_r, loc=<-17.8549 0 0>, rot=<0 0 0 -1>}
 cr_lowerarm_push_r {parent=lowerarm_r, loc=<0 0 0>, rot=<0 0 0 1>}
 hand_r_null {parent=upperarm_r, type=Null, loc=<-51.6459 -12.7131 1.2215>, rot=<0.6448 0.2741 0.0395 0.7124>}
 deform_clavicle_l {parent=spine_05, loc=<2.5856 -0.2866 -4.3777>, rot=<0.0465 -0.7289 -0.009 -0.6829>}
 deform_upperarm_l {parent=deform_clavicle_l, loc=<18.2264 0.1342 -0.001>, rot=<0.0253 -0.2693 0.0013 -0.9627>}
 deform_lowerarm_l {parent=deform_upperarm_l, loc=<29.0071 0.1686 -1.239>, rot=<0.0391 -0.0099 -0.2453 -0.9686>}
 deform_hand_l {parent=deform_lowerarm_l, loc=<25.8607 0.222 1.0274>, rot=<-0.7193 -0.0988 0.1193 -0.6773>}
 deform_hand_up_l {parent=deform_lowerarm_l, loc=<25.8607 0.222 1.0274>, rot=<-0.7269 0 0 -0.6867>}
 deform_hand_dw_l {parent=deform_lowerarm_l, loc=<25.8607 0.222 1.0274>, rot=<-0.7269 0 0 -0.6867>}
 deform_lowerarm_twist_02_l {parent=deform_lowerarm_l, loc=<9.6199 0 0>, rot=<0 0 0 -1>}
 deform_lowerarm_twist_01_l {parent=deform_lowerarm_l, loc=<17.8549 0 0>, rot=<0 0 0 -1>}
 deform_upperarm_l_end {parent=deform_upperarm_l, loc=<29.0071 0.1686 -1.239>, rot=<0.0391 -0.0099 -0.2453 -0.9686>}
 deform_upperarm_twist_02_l {parent=deform_upperarm_l, loc=<19.5059 0.0843 -0.6195>, rot=<0 0 0 -1>}
 deform_upperarm_twist_01_l {parent=deform_upperarm_l, loc=<7.9796 0.0843 -0.6195>, rot=<0 0 0 -1>}
 deform_clavicle_r {parent=spine_05, loc=<2.5855 -0.2866 4.3777>, rot=<0.7289 0.0465 -0.6829 0.009>}
 deform_upperarm_r {parent=deform_clavicle_r, loc=<-18.2264 -0.1342 0.001>, rot=<0.0253 -0.2693 0.0013 -0.9627>}
 deform_lowerarm_r {parent=deform_upperarm_r, loc=<-29.0071 -0.1686 1.239>, rot=<0.0391 -0.0099 -0.2453 -0.9686>}
 deform_hand_r {parent=deform_lowerarm_r, loc=<-25.8607 -0.222 -1.0274>, rot=<-0.7193 -0.0988 0.1193 -0.6773>}
 deform_lowerarm_twist_02_r {parent=deform_lowerarm_r, loc=<-9.6199 0 0>, rot=<0 0 0 -1>}
 deform_lowerarm_twist_01_r {parent=deform_lowerarm_r, loc=<-17.8549 0 0>, rot=<0 0 0 -1>}
 deform_hand_up_r {parent=deform_lowerarm_r, loc=<-25.8607 -0.222 -1.0274>, rot=<-0.7269 0 0 -0.6867>}
 deform_hand_dw_r {parent=deform_lowerarm_r, loc=<-25.8607 -0.222 -1.0274>, rot=<-0.7269 0 0 -0.6867>}
 deform_upperarm_twist_01_r {parent=deform_upperarm_r, loc=<-7.9796 -0.0843 0.6195>, rot=<0 0 0 -1>}
 deform_upperarm_twist_02_r {parent=deform_upperarm_r, loc=<-19.5059 -0.0843 0.6195>, rot=<0 0 0 -1>}
 deform_upperarm_r_end {parent=deform_upperarm_r, loc=<-29.0071 -0.1686 1.239>, rot=<0.0391 -0.0099 -0.2453 -0.9686>}
 deform_collar_01_back_l {parent=spine_05, loc=<14.9165 -7.6944 -12.2194>, rot=<0.8983 -0.0993 -0.0978 -0.4167>}
 deform_collar_02_back_l {parent=deform_collar_01_back_l, loc=<6.6606 0 0>, rot=<-0.0376 0.0134 -0.2345 -0.9713>}
 dyn_collar_01_back_l {parent=deform_collar_02_back_l, loc=<0 0 0>, rot=<0 0 0 1>}
 dyn_collar_02_back_l {parent=dyn_collar_01_back_l, loc=<10.2454 0 0>, rot=<0 0 0 1>}
 deform_collar_01_back {parent=spine_05, loc=<11.1938 -10.4791 0>, rot=<0.9969 -0.0784 0 0>}
 dyn_collar_01_back {parent=deform_collar_01_back, loc=<7.0488 0 0>, rot=<0 0 0 1>}
 deform_collar_01_back_r {parent=spine_05, loc=<14.9166 -7.6944 12.2194>, rot=<-0.0993 -0.8983 0.4167 -0.0978>}
 deform_collar_02_back_r {parent=deform_collar_01_back_r, loc=<-6.6609 0.0001 0>, rot=<-0.0376 0.0134 -0.2345 -0.9713>}
 dyn_collar_01_back_r {parent=deform_collar_02_back_r, loc=<0 0 0>, rot=<0 0 0 1>}
 dyn_collar_02_back_r {parent=dyn_collar_01_back_r, loc=<-10.2452 -0.0003 -0.0001>, rot=<0 0 0 1>}
 cape_l_MSTR {parent=spine_05, loc=<8.4864 -8.223 -23.2637>, rot=<0.0547 -0.6927 0.7035 -0.1489>}
 cape_01_l {parent=cape_l_MSTR, loc=<0 0 0>, rot=<0 0 0 1>}
 cape_02_l {parent=cape_01_l, loc=<14.7369 0 0>, rot=<0 -0.0808 0 -0.9967>}
 cape_03_l {parent=cape_02_l, loc=<29.2632 0 0>, rot=<0.0851 0.0004 0.0047 -0.9964>}
 cape_04_l {parent=cape_03_l, loc=<25.8096 0 0>, rot=<0.4234 0 0 -0.906>}
 cape_05_l {parent=cape_04_l, loc=<20.5355 0.0001 -0.0001>, rot=<-0.3273 0 0 -0.9449>}
 cape_06_l {parent=cape_05_l, loc=<14.1389 0.0001 0>, rot=<0 0 0 1>}
 cape_r_MSTR {parent=spine_05, loc=<8.4862 -8.223 23.2637>, rot=<0.1489 -0.7036 0.6927 -0.0547>}
 cape_01_r {parent=cape_r_MSTR, loc=<0 0 0>, rot=<0 0 0 1>}
 cape_02_r {parent=cape_01_r, loc=<14.7371 0 0>, rot=<0 -0.0808 0 -0.9967>}
 cape_03_r {parent=cape_02_r, loc=<29.2629 0 0>, rot=<-0.0849 0.0004 -0.0047 -0.9964>}
 cape_04_r {parent=cape_03_r, loc=<25.8097 0 0>, rot=<-0.5175 0 0 -0.8557>}
 cape_05_r {parent=cape_04_r, loc=<20.5355 -0.0001 -0.0002>, rot=<0.5266 0 0 -0.8501>}
 cape_06_r {parent=cape_05_r, loc=<14.1389 -0.0001 0>, rot=<0 0 0 1>}
 deform_uppershoulder_r {parent=spine_05, loc=<6.4945 -0.4746 20.6521>, rot=<0.7289 0.0465 -0.6829 0.009>}
 deform_chestsqueeze_r {parent=deform_uppershoulder_r, loc=<-2.6115 -15.0359 4.4204>, rot=<0.855 -0.0806 0.3041 -0.4123>}
 deform_shouldersqueeze_r {parent=deform_uppershoulder_r, loc=<-7.0592 0.3197 0.4253>, rot=<0 0 0 1>}
 deform_shouldersqueeze_front_r {parent=deform_shouldersqueeze_r, loc=<0.4076 -8.0991 0.0274>, rot=<0 0 0 1>}
 deform_chestsqueeze {parent=deform_uppershoulder_r, loc=<-2.6115 -15.0359 4.4204>, rot=<0.855 -0.0806 0.3041 -0.4123>}
 deform_shouldersqueeze {parent=deform_uppershoulder_r, loc=<-7.0592 0.3197 0.4253>, rot=<0 0 0 1>}
 cape_mid_01_l_MSTR {parent=spine_05, loc=<7.8598 -11.9799 -9.1594>, rot=<0.1426 -0.6935 0.689 -0.1551>}
 cape_mid_01_l {parent=cape_mid_01_l_MSTR, loc=<0 0 0>, rot=<0 0 0 1>}
 cape_mid_02_l {parent=cape_mid_01_l, loc=<15.6782 0 0>, rot=<0 -0.075 0 -0.9972>}
 cape_mid_03_l {parent=cape_mid_02_l, loc=<28.1124 0 0>, rot=<0 -0.0196 0 -0.9998>}
 cape_mid_04_l {parent=cape_mid_03_l, loc=<25.1969 0 0>, rot=<0 -0.0051 0 -1>}
 cape_mid_05_l {parent=cape_mid_04_l, loc=<22.2027 0 0>, rot=<-0.0044 0 0 -1>}
 cape_mid_06_l {parent=cape_mid_05_l, loc=<13.3811 0 0>, rot=<0 0 0 1>}
 cape_mid_01_r_MSTR {parent=spine_05, loc=<7.8603 -11.9799 9.1594>, rot=<0.1551 -0.689 0.6935 -0.1426>}
 cape_mid_01_r {parent=cape_mid_01_r_MSTR, loc=<0 0 0>, rot=<0 0 0 1>}
 cape_mid_02_r {parent=cape_mid_01_r, loc=<15.6785 0 0>, rot=<0 -0.075 0 -0.9972>}
 cape_mid_03_r {parent=cape_mid_02_r, loc=<28.1126 0 0>, rot=<0 -0.0196 0 -0.9998>}
 cape_mid_04_r {parent=cape_mid_03_r, loc=<25.1969 0 0>, rot=<-0.0002 -0.0051 0 -1>}
 cape_mid_05_r {parent=cape_mid_04_r, loc=<22.2027 0 0>, rot=<-0.0235 0 0 -0.9997>}
 cape_mid_06_r {parent=cape_mid_05_r, loc=<13.3811 0 0>, rot=<0 0 0 1>}
 deform_uppershoulder_l {parent=spine_05, loc=<6.4949 -0.4746 -20.6521>, rot=<0.0465 -0.7289 -0.009 -0.6829>}
 deform_chestsqueeze_l {parent=deform_uppershoulder_l, loc=<2.6116 15.0358 -4.4208>, rot=<0.855 -0.0806 0.3041 -0.4123>}
 deform_shouldersqueeze_l {parent=deform_uppershoulder_l, loc=<7.0593 -0.3197 -0.4254>, rot=<0 0 0 1>}
 deform_shouldersqueeze_front_l {parent=deform_shouldersqueeze_l, loc=<-0.4076 8.0991 -0.0274>, rot=<0 0 0 1>}
 deform_chestsqueeze_target_r {parent=spine_04, loc=<0.0555 23.7308 3.8015>, rot=<0.7289 0.0476 -0.6829 0.01>}
 deform_chestsqueeze_target {parent=spine_04, loc=<0.0555 23.7308 3.8015>, rot=<0.7289 0.0476 -0.6829 0.01>}
 thigh_l {parent=pelvis, loc=<-1.8716 -0.0109 -11.9936>, rot=<0.0012 0.0572 0.0155 -0.9982>}
 calf_l {parent=thigh_l, loc=<-41.9178 -0.2246 0.0926>, rot=<0.0006 -0.0161 -0.0391 -0.9991>}
 foot_l {parent=calf_l, loc=<-38.7057 1.2438 0.2452>, rot=<0.0975 -0.0474 0.0317 -0.9936>}
 ball_l {parent=foot_l, loc=<-6.1614 13.1469 0.6807>, rot=<0.0001 -0.0001 -0.7063 -0.7079>}
 calf_twist_02_l {parent=calf_l, loc=<-15.9268 -0.0967 0.0397>, rot=<0.0237 0.0095 -0.0035 -0.9997>}
 calf_twist_01_l {parent=calf_l, loc=<-32.9335 -0.0967 0.0397>, rot=<0.0237 0.0095 -0.0035 -0.9997>}
 thigh_twist_01_l {parent=thigh_l, loc=<-21.6303 0 0>, rot=<0.0013 0 0 -1>}
 thigh_r {parent=pelvis, loc=<-1.8716 -0.0109 11.9936>, rot=<0.0572 -0.0012 0.9982 0.0155>}
 calf_r {parent=thigh_r, loc=<41.9178 0.2246 -0.0926>, rot=<0.0006 -0.0161 -0.0391 -0.9991>}
 foot_r {parent=calf_r, loc=<38.7057 -1.2438 -0.2452>, rot=<0.0975 -0.0474 0.0317 -0.9936>}
 ball_r {parent=foot_r, loc=<6.1614 -13.1469 -0.6807>, rot=<0.0001 -0.0001 -0.7063 -0.7079>}
 calf_twist_02_r {parent=calf_r, loc=<15.9268 0.0967 -0.0397>, rot=<0.0237 0.0095 -0.0035 -0.9997>}
 calf_twist_01_r {parent=calf_r, loc=<32.9335 0.0967 -0.0397>, rot=<0.0237 0.0095 -0.0035 -0.9997>}
 thigh_twist_01_r {parent=thigh_r, loc=<21.6303 0 0>, rot=<0.0013 0 0 -1>}
 deform_thigh_r {parent=pelvis, loc=<-6.0469 0.1183 12.4736>, rot=<0.0572 -0.0012 0.9982 0.0155>}
 deform_thigh_l {parent=pelvis, loc=<-6.0487 0.1184 -12.4738>, rot=<0.0012 0.0572 0.0155 -0.9982>}
 deform_thighpush_base_l {parent=pelvis, loc=<-6.0487 0.1184 -12.4738>, rot=<0.0012 0.0572 0.0155 -0.9982>}
 deform_thighpush_l {parent=deform_thighpush_base_l, loc=<0 0 0>, rot=<0 0 0 1>}
 deform_thighpush_base_r {parent=pelvis, loc=<-6.0469 0.1183 12.4736>, rot=<0.0572 -0.0012 0.9982 0.0155>}
 deform_thighpush_r {parent=deform_thighpush_base_r, loc=<0 0 0>, rot=<0 0 0 1>}
 dyn_cape_groundcollision {parent=root, loc=<0 0 0>, rot=<0 0 0 1>}
]
variables [
 Is SkyDiving {type=bool, default=False}
 L neckside Alpha {type=double, default=0}
 L thighfront Alpha {type=double, default=0}
 R neckside Alpha {type=double, default=0}
 R thighfront Alpha {type=double, default=0}
 RightArm_is_HidenIsWearingGauntlet_alpha_R {type=double, default=0}
 neckback Alpha {type=double, default=0}
]
