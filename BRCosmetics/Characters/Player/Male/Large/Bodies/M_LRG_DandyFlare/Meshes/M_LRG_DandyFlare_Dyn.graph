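# Physics asset for the cape dynamics of the large male DandyFlare player body cosmetic.
# Four six-segment capsule chains hang from spine_05, the middle pair via small MSTR bodies, with tight twist and uneven swing limits.
# Angular-free cross constraints tie the chains together around segments four and five.
# Torso, upper-arm and leg capsules give body shapes, and a huge offset sphere serves as ground collision.
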PhysicsAsset /BRCosmetics/Characters/Player/Male/Large/Bodies/M_LRG_DandyFlare/Meshes/M_LRG_DandyFlare_Dyn
bodies {
 calf_l [{shapeType=Capsule, length=11.7963, radius=42, center=<-27.2085 11.0953 -0.0371>, rotation=<rot 90 -1.4887 0>}]
 calf_r [{shapeType=Capsule, length=11.7963, radius=42, center=<27.2085 -11.0953 0.0371>, rotation=<rot 90 178.5113 0>}]
 cape_01_l [{shapeType=Capsule, length=13.5, radius=10, center=<9.4584 0 11.5842>, rotation=<rot -90 0 0>}]
 cape_01_r [{shapeType=Capsule, length=13.5, radius=10, center=<9.4584 0 11.5842>, rotation=<rot -90 0 0>}]
 cape_02_l [{shapeType=Capsule, length=13.5, radius=10, center=<9.0802 0.0111 11.6654>, rotation=<rot 90 0 0>}]
 cape_02_r [{shapeType=Capsule, length=13.5, radius=10, center=<9.0802 -0.0112 11.6654>, rotation=<rot 90 0 0>}]
 cape_03_l [{shapeType=Capsule, length=14, radius=12, center=<12 0 10>, rotation=<rot 90 0 0>}]
 cape_03_r [{shapeType=Capsule, length=14, radius=12, center=<12 0 10>, rotation=<rot 90 0 0>}]
 cape_04_l [{shapeType=Capsule, length=14, radius=12.723, center=<9 0 5.6542>, rotation=<rot 90 0 0>}]
 cape_04_r [{shapeType=Capsule, length=14, radius=12.723, center=<9 0 5.4085>, rotation=<rot 90 0 0>}]
 cape_05_l [{shapeType=Capsule, length=14, radius=12.723, center=<9 0 13.502>, rotation=<rot 90 0 0>}]
 cape_05_r [{shapeType=Capsule, length=14, radius=12.723, center=<9 0 13.502>, rotation=<rot 90 0 0>}]
 cape_06_l [{shapeType=Capsule, length=7, radius=13, center=<2.5881 0 13.502>, rotation=<rot 90 0 0>}]
 cape_06_r [{shapeType=Capsule, length=7, radius=13, center=<2.5881 0 13.502>, rotation=<rot 90 0 0>}]
 cape_mid_01_l [{shapeType=Capsule, length=12, radius=10, center=<9 0 10>, rotation=<rot 90 0 0>}]
 cape_mid_01_l_MSTR [{shapeType=Capsule, length=1, radius=1}]
 cape_mid_01_r [{shapeType=Capsule, length=12, radius=10, center=<9 0 10>, rotation=<rot 90 0 0>}]
 cape_mid_01_r_MSTR [{shapeType=Capsule, length=1, radius=1}]
 cape_mid_02_l [{shapeType=Capsule, length=13.5, radius=12, center=<12 0 10>, rotation=<rot 90 0 0>}]
 cape_mid_02_r [{shapeType=Capsule, length=13.5, radius=12, center=<12 0 10>, rotation=<rot 90 0 0>}]
 cape_mid_03_l [{shapeType=Capsule, length=14, radius=12.723, center=<9 0 7>, rotation=<rot 90 0 0>}]
 cape_mid_03_r [{shapeType=Capsule, length=14, radius=12.723, center=<9 0 7>, rotation=<rot 90 0 0>}]
 cape_mid_04_l [{shapeType=Capsule, length=14, radius=12.723, center=<9 0 13.502>, rotation=<rot 90 0 0>}]
 cape_mid_04_r [{shapeType=Capsule, length=14, radius=12.723, center=<9 0 13.502>, rotation=<rot 90 0 0>}]
 cape_mid_05_l [{shapeType=Capsule, length=14, radius=12.723, center=<9 0 13.502>, rotation=<rot 90 0 0>}]
 cape_mid_05_r [{shapeType=Capsule, length=14, radius=12.723, center=<9 0 13.502>, rotation=<rot 90 0 0>}]
 cape_mid_06_l [{shapeType=Capsule, length=6, radius=13, center=<2.5881 0 13.502>, rotation=<rot 90 0 0>}]
 cape_mid_06_r [{shapeType=Capsule, length=6, radius=13, center=<2.5881 0 13.502>, rotation=<rot 90 0 0>}]
 dyn_cape_groundcollision [{shapeType=Sphere, radius=6000, shapeName=dyn_cape_groundcollision_sphere, center=<0 0 -6000>}]
 pelvis [{shapeType=Capsule, length=4.5938, radius=21.5, center=<-0.9978 -2.5693 0>}]
 spine_01 [{shapeType=Capsule, length=4.0312, radius=12, center=<3.5949 1 0>}]
 spine_02 [{shapeType=Capsule, length=1.3125, radius=16.5, center=<3.5949 0 0>}]
 spine_03 [{shapeType=Capsule, length=39.6562, radius=12, center=<2.5875 -1 0>}]
 spine_04 [{shapeType=Capsule, length=29.9062, radius=17.5, center=<2.8832 -0.1618 0>}]
 spine_05 [{shapeType=Capsule, length=32.875, radius=16, center=<-0.0029 0.1172 0>}]
 thigh_l [{shapeType=Capsule, length=28.0938, radius=21, center=<-19.7778 -0.1788 0>, rotation=<rot 90 18.9704 16.3433>}]
 thigh_r [{shapeType=Capsule, length=28.0938, radius=21, center=<19.7778 0.1788 0>, rotation=<rot 90 -10.6197 166.7532>}]
 upperarm_l [{shapeType=Capsule, length=20.625, radius=6.9375, center=<11.9672 -1.662 0>, rotation=<rot 90 0 -6.5611>}]
 upperarm_r [{shapeType=Capsule, length=20.625, radius=6.9375, center=<-11.9672 1.662 0>, rotation=<rot 90 0 173.4389>}]
}
constraints [
 {
  bone1Name=spine_01
  bone2Name=pelvis
  swing1LimitDegrees=45
  swing1Motion=Free
  swing2LimitDegrees=45
  swing2Motion=Free
  twistLimitDegrees=45
  twistMotion=Free
 }
 {
  bone1Name=spine_02
  bone2Name=spine_01
  swing1LimitDegrees=45
  swing1Motion=Free
  swing2LimitDegrees=45
  swing2Motion=Free
  twistLimitDegrees=45
  twistMotion=Free
 }
 {
  bone1Name=spine_03
  bone2Name=spine_02
  swing1LimitDegrees=45
  swing1Motion=Free
  swing2LimitDegrees=45
  swing2Motion=Free
  twistLimitDegrees=45
  twistMotion=Free
 }
 {
  bone1Name=spine_04
  bone2Name=spine_03
  swing1LimitDegrees=45
  swing1Motion=Free
  swing2LimitDegrees=45
  swing2Motion=Free
  twistLimitDegrees=45
  twistMotion=Free
 }
 {
  bone1Name=spine_05
  bone2Name=spine_04
  swing1LimitDegrees=45
  swing1Motion=Free
  swing2LimitDegrees=45
  swing2Motion=Free
  twistLimitDegrees=45
  twistMotion=Free
 }
 {
  bone1Name=cape_mid_01_l_MSTR
  bone2Name=spine_05
  swing1LimitDegrees=45
  swing1Motion=Free
  swing2LimitDegrees=45
  swing2Motion=Free
  twistLimitDegrees=45
  twistMotion=Free
 }
 {bone1Name=cape_mid_01_l, bone2Name=cape_mid_01_l_MSTR, swing1LimitDegrees=5, swing2LimitDegrees=45, twistLimitDegrees=5}
 {
  bone1Name=cape_mid_01_r_MSTR
  bone2Name=spine_05
  swing1LimitDegrees=45
  swing1Motion=Free
  swing2LimitDegrees=45
  swing2Motion=Free
  twistLimitDegrees=45
  twistMotion=Free
 }
 {bone1Name=cape_mid_01_r, bone2Name=cape_mid_01_r_MSTR, swing1LimitDegrees=5, swing2LimitDegrees=45, twistLimitDegrees=5}
 {bone1Name=cape_01_l, bone2Name=spine_05, swing1LimitDegrees=5, swing2LimitDegrees=45, twistLimitDegrees=5}
 {bone1Name=cape_02_l, bone2Name=cape_01_l, swing1LimitDegrees=5, swing2LimitDegrees=45, twistLimitDegrees=5}
 {bone1Name=cape_03_l, bone2Name=cape_02_l, swing1LimitDegrees=5, swing2LimitDegrees=45, twistLimitDegrees=5}
 {bone1Name=cape_04_l, bone2Name=cape_03_l, swing1LimitDegrees=10, swing2LimitDegrees=45, twistLimitDegrees=1}
 {bone1Name=cape_05_l, bone2Name=cape_04_l, swing1LimitDegrees=10, swing2LimitDegrees=45, twistLimitDegrees=1}
 {bone1Name=cape_06_l, bone2Name=cape_05_l, swing1LimitDegrees=10, swing2LimitDegrees=45, twistLimitDegrees=1}
 {bone1Name=cape_01_r, bone2Name=spine_05, swing1LimitDegrees=5, swing2LimitDegrees=45, twistLimitDegrees=5}
 {bone1Name=cape_02_r, bone2Name=cape_01_r, swing1LimitDegrees=5, swing2LimitDegrees=45, twistLimitDegrees=5}
 {bone1Name=cape_03_r, bone2Name=cape_02_r, swing1LimitDegrees=5, swing2LimitDegrees=45, twistLimitDegrees=5}
 {bone1Name=cape_04_r, bone2Name=cape_03_r, swing1LimitDegrees=10, swing2LimitDegrees=45, twistLimitDegrees=1}
 {bone1Name=cape_05_r, bone2Name=cape_04_r, swing1LimitDegrees=10, swing2LimitDegrees=45, twistLimitDegrees=1}
 {bone1Name=cape_06_r, bone2Name=cape_05_r, swing1LimitDegrees=10, swing2LimitDegrees=45, twistLimitDegrees=1}
 {
  bone1Name=thigh_l
  bone2Name=pelvis
  swing1LimitDegrees=45
  swing1Motion=Free
  swing2LimitDegrees=45
  swing2Motion=Free
  twistLimitDegrees=45
  twistMotion=Free
 }
 {
  bone1Name=calf_l
  bone2Name=thigh_l
  swing1LimitDegrees=45
  swing1Motion=Free
  swing2LimitDegrees=45
  swing2Motion=Free
  twistLimitDegrees=45
  twistMotion=Free
 }
 {
  bone1Name=upperarm_l
  bone2Name=spine_05
  swing1LimitDegrees=45
  swing1Motion=Free
  swing2LimitDegrees=45
  swing2Motion=Free
  twistLimitDegrees=45
  twistMotion=Free
 }
 {
  bone1Name=upperarm_r
  bone2Name=spine_05
  swing1LimitDegrees=45
  swing1Motion=Free
  swing2LimitDegrees=45
  swing2Motion=Free
  twistLimitDegrees=45
  twistMotion=Free
 }
 {
  bone1Name=calf_r
  bone2Name=thigh_r
  swing1LimitDegrees=45
  swing1Motion=Free
  swing2LimitDegrees=45
  swing2Motion=Free
  twistLimitDegrees=45
  twistMotion=Free
 }
 {
  bone1Name=thigh_r
  bone2Name=pelvis
  swing1LimitDegrees=45
  swing1Motion=Free
  swing2LimitDegrees=45
  swing2Motion=Free
  twistLimitDegrees=45
  twistMotion=Free
 }
 {bone1Name=cape_mid_02_l, bone2Name=cape_mid_01_l, swing1LimitDegrees=5, swing2LimitDegrees=45, twistLimitDegrees=5}
 {bone1Name=cape_mid_03_l, bone2Name=cape_mid_02_l, swing1LimitDegrees=10, swing2LimitDegrees=45, twistLimitDegrees=1}
 {bone1Name=cape_mid_04_l, bone2Name=cape_mid_03_l, swing1LimitDegrees=10, swing2LimitDegrees=45, twistLimitDegrees=1}
 {bone1Name=cape_mid_05_l, bone2Name=cape_mid_04_l, swing1LimitDegrees=10, swing2LimitDegrees=45, twistLimitDegrees=1}
 {bone1Name=cape_mid_06_l, bone2Name=cape_mid_05_l, swing1LimitDegrees=10, swing2LimitDegrees=45, twistLimitDegrees=1}
 {bone1Name=cape_mid_02_r, bone2Name=cape_mid_01_r, swing1LimitDegrees=5, swing2LimitDegrees=45, twistLimitDegrees=5}
 {bone1Name=cape_mid_03_r, bone2Name=cape_mid_02_r, swing1LimitDegrees=10, swing2LimitDegrees=45, twistLimitDegrees=1}
 {bone1Name=cape_mid_04_r, bone2Name=cape_mid_03_r, swing1LimitDegrees=10, swing2LimitDegrees=45, twistLimitDegrees=1}
 {bone1Name=cape_mid_05_r, bone2Name=cape_mid_04_r, swing1LimitDegrees=10, swing2LimitDegrees=45, twistLimitDegrees=1}
 {bone1Name=cape_mid_06_r, bone2Name=cape_mid_05_r, swing1LimitDegrees=10, swing2LimitDegrees=45, twistLimitDegrees=1}
 {
  bone1Name=cape_mid_04_l
  bone2Name=cape_mid_04_r
  swing1LimitDegrees=45
  swing1Motion=Free
  swing2LimitDegrees=45
  swing2Motion=Free
  twistLimitDegrees=45
  twistMotion=Free
 }
 {
  bone1Name=cape_mid_04_r
  bone2Name=cape_05_r
  swing1LimitDegrees=45
  swing1Motion=Free
  swing2LimitDegrees=45
  swing2Motion=Free
  twistLimitDegrees=45
  twistMotion=Free
 }
 {
  bone1Name=cape_05_l
  bone2Name=cape_mid_04_l
  swing1LimitDegrees=45
  swing1Motion=Free
  swing2LimitDegrees=45
  swing2Motion=Free
  twistLimitDegrees=45
  twistMotion=Free
 }
]
