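# Physics asset for the large male DandyFlare player body cosmetic, giving its skeleton simplified collision shapes.
# Capsules cover the pelvis, five spine bones, thighs and calves, with no bodies for arms, head or feet.
# An oversized root sphere sits far below the origin, its top just above it like a curved floor.
# Constraints chain spine and legs through the pelvis to root, with swing and twist left free.
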PhysicsAsset /BRCosmetics/Characters/Player/Male/Large/Bodies/M_LRG_DandyFlare/Meshes/M_LRG_DandyFlare_Physics
bodies {
 calf_l [{shapeType=Capsule, length=35, radius=20, shapeName=calf_l_capsule, center=<-23.2947 5.6802 1.7373>, rotation=<rot 0 -90.0001 -86.2828>}]
 calf_r [{shapeType=Capsule, length=35, radius=20, shapeName=calf_l_capsule, center=<23.2947 -5.6802 0.1437>, rotation=<rot 0 -90.0001 -86.2828>}]
 pelvis [{shapeType=Capsule, length=6.9574, radius=16.227, center=<-2.4887 -1.6263 0>}]
 root [{shapeType=Sphere, radius=298, shapeName=root_sphere, center=<0 -24.5697 -289.4702>}]
 spine_01 [{shapeType=Capsule, length=7.9438, radius=13.5586, center=<-1.7965 0.1884 0>}]
 spine_02 [{shapeType=Capsule, length=5.7999, radius=14.1068, center=<-1.711 -0.1851 0.0001>}]
 spine_03 [{shapeType=Capsule, length=6.6842, radius=14.1776, center=<0.1445 -1.3241 0.0001>}]
 spine_04 [{shapeType=Capsule, length=7.9139, radius=14.5355, center=<4.2405 -1.1604 0.0002>}]
 spine_05 [{shapeType=Capsule, length=22, radius=25, center=<-5.0021 8.5453 0>}]
 thigh_l [{
   shapeType=Capsule
   length=28.4881
   radius=12.7772
   shapeName=calf_l_capsule
   center=<-14.476 -1.9053 0.9405>
   rotation=<rot 0 -90.0001 -86.2828>
  }]
 thigh_r [{
   shapeType=Capsule
   length=28.4881
   radius=12.7772
   shapeName=calf_l_capsule
   center=<14.476 1.9053 0.9405>
   rotation=<rot 0 -90.0001 -86.2828>
  }]
}
constraints [
 {
  bone1Name=spine_05
  bone2Name=spine_04
  swing1LimitDegrees=45
  swing1Motion=Free
  swing2LimitDegrees=45
  swing2Motion=Free
  twistLimitDegrees=45
  twistMotion=Free
 }
 {
  bone1Name=spine_01
  bone2Name=pelvis
  swing1LimitDegrees=45
  swing1Motion=Free
  swing2LimitDegrees=45
  swing2Motion=Free
  twistLimitDegrees=45
  twistMotion=Free
 }
 {
  bone1Name=spine_02
  bone2Name=spine_01
  swing1LimitDegrees=45
  swing1Motion=Free
  swing2LimitDegrees=45
  swing2Motion=Free
  twistLimitDegrees=45
  twistMotion=Free
 }
 {
  bone1Name=spine_03
  bone2Name=spine_02
  swing1LimitDegrees=45
  swing1Motion=Free
  swing2LimitDegrees=45
  swing2Motion=Free
  twistLimitDegrees=45
  twistMotion=Free
 }
 {
  bone1Name=spine_04
  bone2Name=spine_03
  swing1LimitDegrees=45
  swing1Motion=Free
  swing2LimitDegrees=45
  swing2Motion=Free
  twistLimitDegrees=45
  twistMotion=Free
 }
 {
  bone1Name=calf_l
  bone2Name=thigh_l
  swing1LimitDegrees=45
  swing1Motion=Free
  swing2LimitDegrees=45
  swing2Motion=Free
  twistLimitDegrees=45
  twistMotion=Free
 }
 {
  bone1Name=calf_r
  bone2Name=thigh_r
  swing1LimitDegrees=45
  swing1Motion=Free
  swing2LimitDegrees=45
  swing2Motion=Free
  twistLimitDegrees=45
  twistMotion=Free
 }
 {
  bone1Name=pelvis
  bone2Name=root
  swing1LimitDegrees=45
  swing1Motion=Free
  swing2LimitDegrees=45
  swing2Motion=Free
  twistLimitDegrees=45
  twistMotion=Free
 }
 {
  bone1Name=thigh_r
  bone2Name=pelvis
  swing1LimitDegrees=45
  swing1Motion=Free
  swing2LimitDegrees=45
  swing2Motion=Free
  twistLimitDegrees=45
  twistMotion=Free
 }
 {
  bone1Name=thigh_l
  bone2Name=pelvis
  swing1LimitDegrees=45
  swing1Motion=Free
  swing2LimitDegrees=45
  swing2Motion=Free
  twistLimitDegrees=45
  twistMotion=Free
 }
]
