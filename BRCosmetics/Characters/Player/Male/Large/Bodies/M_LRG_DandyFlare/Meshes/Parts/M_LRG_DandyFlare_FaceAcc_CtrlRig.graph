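# Control Rig for the DandyFlare large male FaceAcc part that damps neck and head rotation and drives a mouth-follow deform bone.
# For neck_01, neck_02 and head, it reads rotation relative to a matching rest-pose null as Euler angles, remaps each axis to a smaller range, then recombines it with the null's rotation.
# A second sequence step parent-constrains deform_mouth_follow to a weighted blend of head, both neck bones and spine_05, weighted mostly to head.
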ControlRigBlueprint /BRCosmetics/Characters/Player/Male/Large/Bodies/M_LRG_DandyFlare/Meshes/Parts/M_LRG_DandyFlare_FaceAcc_CtrlRig
graphs [
 RigVMFunctionLibrary {links [], nodes []}
 RigVMModel {nodes [
   RigUnit_BeginExecution = &RigUnit_BeginExecution()
   RigVMFunction_Sequence = &RigVMFunction_Sequence(ExecuteContext=@RigUnit_BeginExecution.ExecutePin)
   &RigUnit_ParentConstraint(
    AdvancedSettings=(RotationOrderForFilter=XZY)
    Child=(Type=Bone,Name="deform_mouth_follow")
    Parents=((Item=(Type=Bone,Name="neck_02"),Weight=0.0141),(Item=(Type=Bone,Name="neck_01"),Weight=0.0776),(Item=(Type=Bone,Name="spine_05"),Weight=0.1006),(Item=(Type=Bone,Name="head"),Weight=0.8077))
    Weight=1
    ExecutePin=@RigVMFunction_Sequence.B
   )
   n1 = &RigVMFunction_MathQuaternionToEuler(
    RotationOrder=ZYX
    Value=&RigUnit_GetRelativeTransformForItem(Child=(Type=Bone,Name="neck_01"), Parent=(Type=Null,Name="neck_01_null"), bChildInitial=False, bParentInitial=False).RelativeTransform.Rotation
   )
   Remap = &RigVMFunction_MathDoubleRemap(SourceMaximum=60, SourceMinimum=-60, TargetMaximum=30, TargetMinimum=-30, Value=@n1.Result.Z)
   RigVMFunction_MathQuaternionFromEuler = &RigVMFunction_MathQuaternionFromEuler(
    Euler=(X=0,Y=0,Z=0)
    Euler.Z=@Remap.Result
    Euler.Y=&RigVMFunction_MathDoubleRemap(SourceMaximum=30, SourceMinimum=-30, TargetMaximum=15, TargetMinimum=-15, Value=@n1.Result.Y).Result
    Euler.X=&RigVMFunction_MathDoubleRemap(SourceMaximum=40, SourceMinimum=-40, TargetMaximum=20, TargetMinimum=-20, Value=@n1.Result.X).Result
   )
   Multiply = /Script/RigVM.RigVMFunction_MathQuaternionMul(
    B=&RigUnit_GetTransform(Item=(Type=Null,Name="neck_01_null"), Space=LocalSpace, bInitial=false).Transform.Rotation
    A=@RigVMFunction_MathQuaternionFromEuler.Result
   ) {is=struct RigVMUnitNode, template="Multiply::Execute(in A,in B,out Result)"}
   RigUnit_SetRotation = &RigUnit_SetRotation(Item=(Type=Bone,Name="neck_01"), Weight=1, bInitial=False, Value=@Multiply.Result, ExecutePin=@RigVMFunction_Sequence.A)
   n2 = &RigVMFunction_MathQuaternionToEuler(
    RotationOrder=ZYX
    Value=&RigUnit_GetRelativeTransformForItem(Child=(Type=Bone,Name="neck_02"), Parent=(Type=Null,Name="neck_02_null"), bChildInitial=False, bParentInitial=False).RelativeTransform.Rotation
   )
   n3 = &RigVMFunction_MathDoubleRemap(SourceMaximum=40, SourceMinimum=-40, TargetMaximum=20, TargetMinimum=-20, Value=@n2.Result.X)
   n4 = &RigVMFunction_MathDoubleRemap(SourceMaximum=70, SourceMinimum=-70, TargetMaximum=30, TargetMinimum=-30, Value=@n2.Result.Z)
   n5 = &RigVMFunction_MathDoubleRemap(SourceMaximum=50, SourceMinimum=-50, TargetMaximum=15, TargetMinimum=-15, Value=@n2.Result.Y)
   > &RigUnit_SetRotation(
    Item=(Type=Bone,Name="neck_02")
    Weight=1
    bInitial=False
    Value=&RigVMFunction_MathQuaternionMul(
     B=&RigUnit_GetTransform(Item=(Type=Null,Name="neck_02_null"), Space=LocalSpace, bInitial=false).Transform.Rotation
     A=&RigVMFunction_MathQuaternionFromEuler(Euler=(X=0,Y=0,Z=0), Euler.X=@n3.Result, Euler.Z=@n4.Result, Euler.Y=@n5.Result).Result
    ) {template="Multiply::Execute(in A,in B,out Result)"}.Result
   )
   n6 = &RigVMFunction_MathQuaternionToEuler(
    RotationOrder=ZYX
    Value=&RigUnit_GetRelativeTransformForItem(Child=(Type=Bone,Name="head"), Parent=(Type=Null,Name="head_null"), bChildInitial=False, bParentInitial=False).RelativeTransform.Rotation
   )
   n7 = &RigVMFunction_MathDoubleRemap(SourceMaximum=30, SourceMinimum=-30, TargetMaximum=10, TargetMinimum=-10, Value=@n6.Result.X)
   n8 = &RigVMFunction_MathDoubleRemap(SourceMaximum=70, SourceMinimum=-70, TargetMaximum=30, TargetMinimum=-30, Value=@n6.Result.Z)
   n9 = &RigVMFunction_MathDoubleRemap(SourceMaximum=30, SourceMinimum=-30, TargetMaximum=15, TargetMinimum=-15, Value=@n6.Result.Y)
   > &RigUnit_SetRotation(
    Item=(Type=Bone,Name="head")
    Weight=1
    bInitial=False
    Value=&RigVMFunction_MathQuaternionMul(
     B=&RigUnit_GetTransform(Item=(Type=Null,Name="head_null"), Space=LocalSpace, bInitial=false).Transform.Rotation
     A=&RigVMFunction_MathQuaternionFromEuler(Euler=(X=0,Y=0,Z=0), Euler.X=@n7.Result, Euler.Z=@n8.Result, Euler.Y=@n9.Result).Result
    ) {template="Multiply::Execute(in A,in B,out Result)"}.Result
   )
  ]}
]
hierarchy [
 root {loc=<0 0 0>, rot=<0 0 0 1>}
 pelvis {parent=root, loc=<0 -0.7321 92.4218>, rot=<0.0127 0.707 -0.0127 -0.707>}
 spine_01 {parent=pelvis, loc=<5.6002 -0.5128 0>, rot=<0 0 0.034 -0.9994>}
 spine_02 {parent=spine_01, loc=<8.0756 0.0139 0>, rot=<0 0 0.0238 -0.9997>}
 spine_03 {parent=spine_02, loc=<7.0395 0 0>, rot=<0 0 0.0149 -0.9999>}
 spine_04 {parent=spine_03, loc=<8.1023 0 0>, rot=<0 0 -0.0132 -0.9999>}
 spine_05 {parent=spine_04, loc=<10.0141 0 0>, rot=<0 0 -0.0015 -1>}
 neck_01 {parent=spine_05, loc=<10.815 1.9196 0>, rot=<0 0 -0.1864 -0.9825>}
 neck_02 {parent=neck_01, loc=<4.7347 0.3045 0>, rot=<0 0 0.0363 -0.9993>}
 head {parent=neck_02, loc=<5.7632 0.5552 0>, rot=<0 0 0.115 -0.9934>}
 deform_mouth_follow {parent=head, loc=<-0.0454 5.4917 0>, rot=<0.0029 -0.7071 -0.0029 -0.7071>}
 faceAttach {parent=head, loc=<0.6419 -1.7133 0>, rot=<0.0029 -0.7071 -0.0029 -0.7071>}
 C_brow_mid {parent=faceAttach, loc=<-0.0036 8.5895 4.6229>, rot=<0 0 0 -1>}
 C_jaw {parent=faceAttach, loc=<-0.1474 2.129 -1.8078>, rot=<0 0 0 1>}
 C_lip_lower_mid {parent=C_jaw, loc=<0.1453 6.9042 -0.6553>, rot=<0 0 0 -1>}
 L_lip_lower_outer {parent=C_jaw, loc=<1.5357 6.3142 -0.4755>, rot=<0.013 0.0445 0.3057 -0.951>}
 R_lip_lower_outer {parent=C_jaw, loc=<-1.3415 6.1623 -0.5263>, rot=<0.013 -0.0445 -0.3057 -0.951>}
 teeth_lower {parent=C_jaw, loc=<0.1284 4.7788 -2.642>, rot=<0 0 0 1>}
 tongue {parent=C_jaw, loc=<0.1658 2.6319 -0.2155>, rot=<-0.1568 0 0 -0.9876>}
 C_lip_upper_mid {parent=faceAttach, loc=<-0.0027 9.2503 -2.3252>, rot=<0.2225 0.0013 -0.0005 -0.9749>}
 C_nose_bridge {parent=faceAttach, loc=<0.0022 10.1193 2.0818>, rot=<-0.2218 0 0 -0.9751>}
 L_brow_mid {parent=faceAttach, loc=<3.249 8.3669 5.5368>, rot=<-0.0001 -0.0042 0.2553 -0.9669>}
 L_brow_outer {parent=faceAttach, loc=<5.0673 6.359 5.4988>, rot=<-0.0287 -0.04 0.4491 -0.8921>}
 L_cheek_inner {parent=faceAttach, loc=<3.8821 6.8452 0.0632>, rot=<-0.0706 -0.2697 0.3184 -0.906>}
 L_eye {parent=faceAttach, loc=<2.7581 6.1306 3.7734>, rot=<0 0 0.9998 0.0175>}
 L_eye_lid_lower_mid {parent=L_eye, loc=<0 0 0>, rot=<0.0329 0 0 -0.9995>}
 L_eye_lid_upper_mid {parent=L_eye, loc=<0 0 0>, rot=<0.0523 0 0 -0.9986>}
 L_lip_corner {parent=faceAttach, loc=<2.2292 7.7865 -2.595>, rot=<0.003 0.008 -0.0146 -0.9999>}
 L_lip_upper_outer {parent=faceAttach, loc=<1.3695 8.5645 -2.3306>, rot=<0.0359 -0.0728 0.2627 -0.9615>}
 R_brow_mid {parent=faceAttach, loc=<-3.2486 8.367 5.5376>, rot=<-0.0001 0.0042 -0.2553 -0.9669>}
 R_brow_outer {parent=faceAttach, loc=<-5.0673 6.359 5.4988>, rot=<-0.0287 0.04 -0.4491 -0.8921>}
 R_cheek_inner {parent=faceAttach, loc=<-3.8821 6.8452 0.0632>, rot=<-0.0706 0.2697 -0.3184 -0.906>}
 R_eye {parent=faceAttach, loc=<-2.7599 6.1306 3.7734>, rot=<0 0 0.9998 -0.0175>}
 R_eye_lid_lower_mid {parent=R_eye, loc=<0 0 0>, rot=<0.0329 0 0 -0.9995>}
 R_eye_lid_upper_mid {parent=R_eye, loc=<0 0 0>, rot=<0.0523 0 0 -0.9986>}
 R_lip_corner {parent=faceAttach, loc=<-2.272 7.7559 -2.5568>, rot=<0.003 -0.008 0.0146 -0.9999>}
 R_lip_upper_outer {parent=faceAttach, loc=<-1.3697 8.5637 -2.33>, rot=<0.0359 0.0728 -0.2627 -0.9615>}
 teeth_upper {parent=faceAttach, loc=<0.0181 6.9804 -0.681>, rot=<0 0 0 1>}
 FX_Eye_L {parent=head, type=Null, loc=<8 1.8 -3.8>, rot=<0 0 0 1>}
 FX_Eye_R {parent=head, type=Null, loc=<8 1.8 3.8>, rot=<0 0 0 1>}
 head_null {parent=neck_02, type=Null, loc=<5.7632 0.5552 0>, rot=<0 0 0.115 -0.9934>}
 neck_02_null {parent=neck_01, type=Null, loc=<4.7347 0.3045 0>, rot=<0 0 0.0363 -0.9993>}
 neck_01_null {parent=spine_05, type=Null, loc=<10.815 1.9196 0>, rot=<0 0 -0.1864 -0.9825>}
 jaw_open_pose {type=Curve, loc=<0 0 0>, rot=<0 0 0 1>}
]
